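# Physics asset for the M_MED_InfernoFNCS male player body cosmetic, built around a jacket with dynamic bones.
# Capsules cover the root, pelvis, legs and five jacket chains: left, right, mid-left, mid-right and back, with longer chains on the left.
# Each chain hangs from a master bone with free rotation on the pelvis, with angular limits along its links.
# Free-rotation cross constraints tie neighboring chains together, linking the panels sideways.
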PhysicsAsset /BRCosmetics/Characters/Player/Male/Medium/Bodies/M_MED_InfernoFNCS/Meshes/M_MED_InfernoFNCS_Physics
bodies {
 calf_l [{shapeType=Capsule, length=30, radius=8, shapeName=calf_l_capsule, center=<-17.8527 0.0001 0>, rotation=<rot 90 0 0.0003>}]
 calf_r [{shapeType=Capsule, length=30, radius=8, shapeName=calf_r_capsule, center=<17.8527 -0.0002 0>, rotation=<rot -90 -90 89.9995>}]
 dyn_jacket_01_L [{shapeType=Capsule, length=4.5, radius=3, shapeName=dyn_jacket_01_L_capsule, center=<6 0 0>, rotation=<rot -90 0 0>}]
 dyn_jacket_01_R [{shapeType=Capsule, length=5, radius=3.7, shapeName=dyn_jacket_01_R_capsule, center=<8.3207 0 0>, rotation=<rot -90 0 0>}]
 dyn_jacket_02_L [{shapeType=Capsule, length=10, radius=2, shapeName=dyn_jacket_02_L_capsule, center=<7.8856 0 0>, rotation=<rot -90 0 0>}]
 dyn_jacket_02_R [{shapeType=Capsule, length=10, radius=2, shapeName=dyn_jacket_02_R_capsule, center=<8.3207 0 0>, rotation=<rot -90 0 0>}]
 dyn_jacket_03_L [{shapeType=Capsule, length=8.5, radius=3.5, shapeName=dyn_jacket_03_L_capsule, center=<10.6199 0 0>, rotation=<rot -90 180 180>}]
 dyn_jacket_03_R [{shapeType=Capsule, length=10, radius=2, shapeName=dyn_jacket_03_R_capsule, center=<8.3207 0 0>, rotation=<rot -90 0 0>}]
 dyn_jacket_04_L [{shapeType=Capsule, length=8.5, radius=3.5, shapeName=dyn_jacket_04_L_capsule, center=<8.3207 0 0>, rotation=<rot -90 0 0>}]
 dyn_jacket_MSTR_L [{shapeType=Capsule, length=0, radius=0.5, shapeName=dyn_jacket_MSTR_L_capsule, rotation=<rot -90 0 0>}]
 dyn_jacket_MSTR_R [{shapeType=Capsule, length=0, radius=0.5, shapeName=dyn_jacket_MSTR_R_capsule, rotation=<rot -90 0 0>}]
 dyn_jacket_bk_01 [{shapeType=Capsule, length=9.5, radius=3.5, shapeName=dyn_jacket_bk_01_capsule, center=<8.3207 0 0>, rotation=<rot -90 0 0>}]
 dyn_jacket_bk_02 [{shapeType=Capsule, length=9, radius=4, shapeName=dyn_jacket_bk_02_capsule, center=<8.1852 0 0>, rotation=<rot -90 0 0>}]
 dyn_jacket_bk_03 [{shapeType=Capsule, length=9, radius=4, shapeName=dyn_jacket_bk_03_capsule, center=<7.1803 0 0>, rotation=<rot -90 180 180>}]
 dyn_jacket_bk_MSTR [{shapeType=Capsule, length=0, radius=0.5, shapeName=dyn_jacket_bk_MSTR_capsule, rotation=<rot -90 0 0>}]
 dyn_jacket_mid_01_L [{shapeType=Capsule, length=8.5, radius=3.5, shapeName=dyn_jacket_mid_01_L_capsule, center=<8.3207 0 0>, rotation=<rot -90 0 0>}]
 dyn_jacket_mid_01_R [{shapeType=Capsule, length=8.5, radius=3.5, shapeName=dyn_jacket_mid_01_R_capsule, center=<8.3207 0 0>, rotation=<rot -90 0 0>}]
 dyn_jacket_mid_02_L [{shapeType=Capsule, length=8, radius=4, shapeName=dyn_jacket_mid_02_L_capsule, center=<8.111 0 0>, rotation=<rot -90 0 0>}]
 dyn_jacket_mid_02_R [{shapeType=Capsule, length=8, radius=4, shapeName=dyn_jacket_mid_02_R_capsule, center=<6.6436 0 0>, rotation=<rot -90 0 0>}]
 dyn_jacket_mid_03_L [{shapeType=Capsule, length=8, radius=4, shapeName=dyn_jacket_mid_03_L_capsule, center=<9.3886 0 0>, rotation=<rot -90 0 0>}]
 dyn_jacket_mid_03_R [{shapeType=Capsule, length=8, radius=4, shapeName=dyn_jacket_mid_03_R_capsule, center=<7.3906 0 0>, rotation=<rot -90 180 180>}]
 dyn_jacket_mid_04_L [{shapeType=Capsule, length=8, radius=4, shapeName=dyn_jacket_mid_04_L_capsule, center=<8.2528 0 0>, rotation=<rot -90 180 180>}]
 dyn_jacket_mid_MSTR_L [{shapeType=Capsule, length=0, radius=0.5, shapeName=dyn_jacket_mid_MSTR_L_capsule, rotation=<rot -90 0 0>}]
 dyn_jacket_mid_MSTR_R [{shapeType=Capsule, length=0, radius=0.5, shapeName=dyn_jacket_mid_MSTR_R_capsule, rotation=<rot -90 0 0>}]
 foot_l [{shapeType=Capsule, length=15, radius=6.5, shapeName=foot_l_capsule, center=<-4.2965 5.4363 0>, rotation=<rot 0 -0.0004 90>}]
 foot_r [{shapeType=Capsule, length=15, radius=6.5, shapeName=foot_r_capsule, center=<4.2964 -5.4364 0>, rotation=<rot 0.0001 179.9991 90>}]
 pelvis [{shapeType=Capsule, length=4, radius=9, shapeName=pelvis_capsule, center=<-2.4945 0 0>}]
 root [{shapeType=Capsule, length=0, radius=26, shapeName=root_capsule, center=<0 0 25.3466>}]
 thigh_l [{shapeType=Capsule, length=26.5, radius=6, shapeName=thigh_l_capsule, center=<-17.4197 0.0002 0>, rotation=<rot 90 0 0.0005>}]
 thigh_r [{shapeType=Capsule, length=26.5, radius=6, shapeName=thigh_r_capsule, center=<17.4197 -0.0001 0>, rotation=<rot -90 180 179.9995>}]
}
constraints [
 {bone1Name=dyn_jacket_01_L, bone2Name=dyn_jacket_MSTR_L, swing1LimitDegrees=0, swing2LimitDegrees=0, twistLimitDegrees=0}
 {bone1Name=dyn_jacket_02_L, bone2Name=dyn_jacket_01_L, swing1LimitDegrees=20, swing2LimitDegrees=45, twistLimitDegrees=5}
 {bone1Name=dyn_jacket_03_L, bone2Name=dyn_jacket_02_L, swing1LimitDegrees=20, swing2LimitDegrees=35, twistLimitDegrees=5}
 {bone1Name=dyn_jacket_04_L, bone2Name=dyn_jacket_03_L, swing1LimitDegrees=20, swing2LimitDegrees=25, twistLimitDegrees=5}
 {bone1Name=dyn_jacket_01_R, bone2Name=dyn_jacket_MSTR_R, swing1LimitDegrees=0, swing2LimitDegrees=0, twistLimitDegrees=0}
 {bone1Name=dyn_jacket_02_R, bone2Name=dyn_jacket_01_R, swing1LimitDegrees=20, swing2LimitDegrees=0, twistLimitDegrees=0}
 {bone1Name=dyn_jacket_03_R, bone2Name=dyn_jacket_02_R, swing1LimitDegrees=20, swing2LimitDegrees=15, twistLimitDegrees=5}
 {bone1Name=dyn_jacket_bk_01, bone2Name=dyn_jacket_bk_MSTR, swing1LimitDegrees=45, swing2LimitDegrees=0, twistLimitDegrees=0}
 {bone1Name=dyn_jacket_bk_02, bone2Name=dyn_jacket_bk_01, swing1LimitDegrees=20, swing2LimitDegrees=45, twistLimitDegrees=5}
 {bone1Name=dyn_jacket_bk_03, bone2Name=dyn_jacket_bk_02, swing1LimitDegrees=20, swing2LimitDegrees=15, twistLimitDegrees=5}
 {bone1Name=dyn_jacket_mid_01_L, bone2Name=dyn_jacket_mid_MSTR_L, swing1LimitDegrees=45, swing2LimitDegrees=0, twistLimitDegrees=0}
 {bone1Name=dyn_jacket_mid_02_L, bone2Name=dyn_jacket_mid_01_L, swing1LimitDegrees=20, swing2LimitDegrees=45, twistLimitDegrees=5}
 {bone1Name=dyn_jacket_mid_03_L, bone2Name=dyn_jacket_mid_02_L, swing1LimitDegrees=20, swing2LimitDegrees=25, twistLimitDegrees=5}
 {bone1Name=dyn_jacket_mid_04_L, bone2Name=dyn_jacket_mid_03_L, swing1LimitDegrees=20, swing2LimitDegrees=15, twistLimitDegrees=5}
 {bone1Name=dyn_jacket_mid_01_R, bone2Name=dyn_jacket_mid_MSTR_R, swing1LimitDegrees=45, swing2LimitDegrees=0, twistLimitDegrees=0}
 {bone1Name=dyn_jacket_mid_02_R, bone2Name=dyn_jacket_mid_01_R, swing1LimitDegrees=20, swing2LimitDegrees=25, twistLimitDegrees=5}
 {bone1Name=dyn_jacket_mid_03_R, bone2Name=dyn_jacket_mid_02_R, swing1LimitDegrees=20, swing2LimitDegrees=15, twistLimitDegrees=5}
 {
  bone1Name=dyn_jacket_mid_MSTR_L
  bone2Name=pelvis
  swing1LimitDegrees=45
  swing1Motion=Free
  swing2LimitDegrees=45
  swing2Motion=Free
  twistLimitDegrees=45
  twistMotion=Free
 }
 {
  bone1Name=dyn_jacket_bk_MSTR
  bone2Name=pelvis
  swing1LimitDegrees=45
  swing1Motion=Free
  swing2LimitDegrees=45
  swing2Motion=Free
  twistLimitDegrees=45
  twistMotion=Free
 }
 {
  bone1Name=dyn_jacket_MSTR_L
  bone2Name=pelvis
  swing1LimitDegrees=45
  swing1Motion=Free
  swing2LimitDegrees=45
  swing2Motion=Free
  twistLimitDegrees=45
  twistMotion=Free
 }
 {
  bone1Name=dyn_jacket_mid_MSTR_R
  bone2Name=pelvis
  swing1LimitDegrees=45
  swing1Motion=Free
  swing2LimitDegrees=45
  swing2Motion=Free
  twistLimitDegrees=45
  twistMotion=Free
 }
 {
  bone1Name=dyn_jacket_MSTR_R
  bone2Name=pelvis
  swing1LimitDegrees=45
  swing1Motion=Free
  swing2LimitDegrees=45
  swing2Motion=Free
  twistLimitDegrees=45
  twistMotion=Free
 }
 {
  bone1Name=thigh_l
  bone2Name=pelvis
  swing1LimitDegrees=45
  swing1Motion=Free
  swing2LimitDegrees=45
  swing2Motion=Free
  twistLimitDegrees=45
  twistMotion=Free
 }
 {
  bone1Name=calf_l
  bone2Name=thigh_l
  swing1LimitDegrees=45
  swing1Motion=Free
  swing2LimitDegrees=45
  swing2Motion=Free
  twistLimitDegrees=45
  twistMotion=Free
 }
 {
  bone1Name=foot_l
  bone2Name=calf_l
  swing1LimitDegrees=45
  swing1Motion=Free
  swing2LimitDegrees=45
  swing2Motion=Free
  twistLimitDegrees=45
  twistMotion=Free
 }
 {
  bone1Name=thigh_r
  bone2Name=pelvis
  swing1LimitDegrees=45
  swing1Motion=Free
  swing2LimitDegrees=45
  swing2Motion=Free
  twistLimitDegrees=45
  twistMotion=Free
 }
 {
  bone1Name=calf_r
  bone2Name=thigh_r
  swing1LimitDegrees=45
  swing1Motion=Free
  swing2LimitDegrees=45
  swing2Motion=Free
  twistLimitDegrees=45
  twistMotion=Free
 }
 {
  bone1Name=foot_r
  bone2Name=calf_r
  swing1LimitDegrees=45
  swing1Motion=Free
  swing2LimitDegrees=45
  swing2Motion=Free
  twistLimitDegrees=45
  twistMotion=Free
 }
 {
  bone1Name=dyn_jacket_04_L
  bone2Name=dyn_jacket_mid_04_L
  swing1LimitDegrees=45
  swing1Motion=Free
  swing2LimitDegrees=45
  swing2Motion=Free
  twistLimitDegrees=45
  twistMotion=Free
 }
 {
  bone1Name=dyn_jacket_mid_02_R
  bone2Name=dyn_jacket_bk_02
  swing1LimitDegrees=45
  swing1Motion=Free
  swing2LimitDegrees=45
  swing2Motion=Free
  twistLimitDegrees=45
  twistMotion=Free
 }
 {
  bone1Name=dyn_jacket_mid_03_L
  bone2Name=dyn_jacket_bk_02
  swing1LimitDegrees=45
  swing1Motion=Free
  swing2LimitDegrees=45
  swing2Motion=Free
  twistLimitDegrees=45
  twistMotion=Free
 }
 {
  bone1Name=dyn_jacket_03_R
  bone2Name=dyn_jacket_mid_03_R
  swing1LimitDegrees=45
  swing1Motion=Free
  swing2LimitDegrees=45
  swing2Motion=Free
  twistLimitDegrees=45
  twistMotion=Free
 }
 {
  bone1Name=dyn_jacket_03_L
  bone2Name=dyn_jacket_mid_03_L
  swing1LimitDegrees=45
  swing1Motion=Free
  swing2LimitDegrees=45
  swing2Motion=Free
  twistLimitDegrees=45
  twistMotion=Free
 }
]
