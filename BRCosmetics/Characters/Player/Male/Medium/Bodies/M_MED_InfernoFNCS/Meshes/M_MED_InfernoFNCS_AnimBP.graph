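# Animation Blueprint for the InfernoFNCS male character body, adding secondary motion to its collar, lapels, holster, pants and other costume bones.
# It copies the attached parent's pose, then applies weighted bone constraints, additive offsets and sprung AnimDynamics with spherical limits.
# The output blends cached main, crouching, sliding and DBNO poses by state bools, with a pelvis-space RigidBody simulation on the main/crouch branch.
# Final lapel offsets fade in during emotes.
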AnimBlueprint /BRCosmetics/Characters/Player/Male/Medium/Bodies/M_MED_InfernoFNCS/Meshes/M_MED_InfernoFNCS_AnimBP
target_skeleton=&/M_MED_InfernoFNCS_Skeleton
graphs [
 AnimGraph {nodes [
   n1 = Anim Dynamics - Bone: dyn_collar_fr_01_L!(ComponentPose, Alpha=1, GravityOverride:Vector="0,0,-980") -> (Pose) {
    angularDampingOverride=1
    angularSpringConstant=20000
    bAngularSpring=true
    bOverrideAngularDamping=true
    bUseGravityOverride=true
    bUseSphericalLimits=true
    gravityScale=0.35
    simulationSpace=RootRelative
    alphaBoolBlend {blendOption=Linear}
    boundBone {boneName=dyn_collar_fr_01_L}
    componentLinearAccScale=<0.7 0.7 0.7>
    componentLinearVelScale=<0.7 0.7 0.7>
    gravityOverride {z=-980}
    physicsBodyDefinitions [{
      collisionType=CustomSphere
      sphereCollisionRadius=2
      boundBone {boneName=dyn_collar_fr_01_L}
      boxExtents=<30 40 40>
      constraintSetup {angularLimitsMax=<0 5 5>, angularLimitsMin=<0 -2 -5>, angularTarget=<1 0 0>}
      localJointOffset=<4 0 0>
     }]
    sphericalLimits [{limitRadius=10, drivingBone {boneName=clavicle_l}, sphereLocalOffset=<10 0 0>}]
   }
   n2 = Anim Dynamics - Bone: dyn_collar_fr_02_L!(ComponentPose=@n1, Alpha=1, GravityOverride:Vector="0,0,-980") -> (Pose) {
    angularDampingOverride=1
    angularSpringConstant=20000
    bAngularSpring=true
    bOverrideAngularDamping=true
    bUseGravityOverride=true
    bUseSphericalLimits=true
    gravityScale=0.35
    simulationSpace=RootRelative
    alphaBoolBlend {blendOption=Linear}
    boundBone {boneName=dyn_collar_fr_02_L}
    componentLinearAccScale=<0.7 0.7 0.7>
    componentLinearVelScale=<0.7 0.7 0.7>
    gravityOverride {z=-980}
    physicsBodyDefinitions [{
      collisionType=CustomSphere
      sphereCollisionRadius=2
      boundBone {boneName=dyn_collar_fr_02_L}
      boxExtents=<30 40 40>
      constraintSetup {angularLimitsMax=<0 5 5>, angularLimitsMin=<0 -2 -5>, angularTarget=<1 0 0>}
      localJointOffset=<4 0 0>
     }]
    sphericalLimits [{limitRadius=10, drivingBone {boneName=clavicle_l}, sphereLocalOffset=<10 0 0>}]
   }
   n3 = Anim Dynamics - Bone: dyn_collar_fr_01_R!(ComponentPose=@n2, Alpha=1, GravityOverride:Vector="0,0,-980") -> (Pose) {
    angularDampingOverride=1
    angularSpringConstant=20000
    bAngularSpring=true
    bOverrideAngularDamping=true
    bUseGravityOverride=true
    bUseSphericalLimits=true
    gravityScale=0.35
    simulationSpace=RootRelative
    alphaBoolBlend {blendOption=Linear}
    boundBone {boneName=dyn_collar_fr_01_R}
    componentLinearAccScale=<0.7 0.7 0.7>
    componentLinearVelScale=<0.7 0.7 0.7>
    gravityOverride {z=-980}
    physicsBodyDefinitions [{
      collisionType=CustomSphere
      sphereCollisionRadius=2
      boundBone {boneName=dyn_collar_fr_01_R}
      boxExtents=<25 35 35>
      constraintSetup {angularLimitsMax=<0 7 3>, angularLimitsMin=<0 0 -5>, angularTarget=<1 0 0>}
      localJointOffset=<4 0 -3>
     }]
    sphericalLimits [{limitRadius=10, drivingBone {boneName=clavicle_r}, sphereLocalOffset=<-8.8576 -4.5635 2.6615>}]
   }
   Anim Dynamics - Bone: dyn_collar_fr_02_R!(ComponentPose=@n3, Alpha=1, GravityOverride:Vector="0,0,-980") -> (Pose) {
    angularDampingOverride=1
    angularSpringConstant=20000
    bAngularSpring=true
    bOverrideAngularDamping=true
    bUseGravityOverride=true
    bUseSphericalLimits=true
    gravityScale=0.35
    simulationSpace=RootRelative
    alphaBoolBlend {blendOption=Linear}
    boundBone {boneName=dyn_collar_fr_02_R}
    componentLinearAccScale=<0.7 0.7 0.7>
    componentLinearVelScale=<0.7 0.7 0.7>
    gravityOverride {z=-980}
    physicsBodyDefinitions [{
      collisionType=CustomSphere
      sphereCollisionRadius=2
      boundBone {boneName=dyn_collar_fr_02_R}
      boxExtents=<25 35 35>
      constraintSetup {angularLimitsMax=<0 7 3>, angularLimitsMin=<0 0 -5>, angularTarget=<1 0 0>}
      localJointOffset=<4 0 -3>
     }]
    sphericalLimits [{limitRadius=10, drivingBone {boneName=clavicle_r}, sphereLocalOffset=<-9.578 -3.737 2.6615>}]
   }
   "Constraint - medallion (P)"!(ComponentPose, Alpha=1, ConstraintWeights_0=1, ConstraintWeights_1:float=0) -> (Pose) {
    alphaBoolBlend {blendOption=Linear}
    boneToModify {boneName=medallion}
    constraintSetup [{targetBone {boneName=dyn_lapel_upper_01_L}}]
    constraintWeights [1,0]
   }
   n4 = "Constraint - dyn_collar_MSTR_L (PP)"!(ComponentPose, Alpha=1, ConstraintWeights_0=0.8, ConstraintWeights_1:float=0.2) -> (Pose) {
    alphaBoolBlend {blendOption=Linear}
    boneToModify {boneName=dyn_collar_MSTR_L}
    constraintSetup [{targetBone {boneName=clavicle_l}}, {targetBone {boneName=spine_05}}]
    constraintWeights [1,0]
   }
   "Constraint - dyn_collar_MSTR_R (PP)"!(ComponentPose=@n4, Alpha=1, ConstraintWeights_0=0.8, ConstraintWeights_1:float=0.2) -> (Pose) {
    alphaBoolBlend {blendOption=Linear}
    boneToModify {boneName=dyn_collar_MSTR_R}
    constraintSetup [{targetBone {boneName=clavicle_r}}, {targetBone {boneName=spine_05}}]
    constraintWeights [1,0]
   }
   n5 = "Constraint - dyn_lapel_upper_MSTR_L (PPPPPPPPPPPPPPPPPPP)"!(
    ComponentPose
    Alpha=1
    ConstraintWeights_0=0.0001
    ConstraintWeights_1:float=0
    ConstraintWeights_2:float=0.0016
    ConstraintWeights_3:float=0.0633
    ConstraintWeights_4:float=0.0774
    ConstraintWeights_5:float=0.8351
    ConstraintWeights_6:float=0
    ConstraintWeights_7:float=0.0003
    ConstraintWeights_8:float=0
    ConstraintWeights_9:float=0
    ConstraintWeights_10:float=0
    ConstraintWeights_11:float=0.0156
    ConstraintWeights_12:float=0
    ConstraintWeights_13:float=0.0055
    ConstraintWeights_14:float=0.001
    ConstraintWeights_15:float=0
    ConstraintWeights_16:float=0
    ConstraintWeights_17:float=0
    ConstraintWeights_18:float=0
   ) -> (Pose) {
    alphaBoolBlend {blendOption=Linear}
    boneToModify {boneName=dyn_lapel_upper_MSTR_L}
    constraintSetup [
     {targetBone {boneName=pelvis}}
     {targetBone {boneName=spine_01}}
     {targetBone {boneName=spine_02}}
     {targetBone {boneName=spine_03}}
     {targetBone {boneName=spine_04}}
     {targetBone {boneName=spine_05}}
     {targetBone {boneName=neck_01}}
     {targetBone {boneName=neck_02}}
     {targetBone {boneName=upperarm_twist_01_l}}
     {targetBone {boneName=upperarm_twist_02_l}}
     {targetBone {boneName=upperarm_twist_01_r}}
     {targetBone {boneName=dyn_lapel_lower_01_L}}
     {targetBone {boneName=dyn_lapel_lower_01_R}}
     {targetBone {boneName=dyn_collar_01_L}}
     {targetBone {boneName=dyn_collar_02_L}}
     {targetBone {boneName=dyn_collar_01_R}}
     {targetBone {boneName=dyn_collar_02_R}}
     {targetBone {boneName=dyn_lapel_upper_01_R}}
     {targetBone {boneName=dyn_collar_bk_01}}
    ]
    constraintWeights [1,1,1,1,1,1,1,1,1,1,1,1,1,1,1,1,1,1,1]
   }
   n6 = "Constraint - dyn_lapel_upper_MSTR_R (PPPPPPPPPPPPPPPPPPPPP)"!(
    ComponentPose=@n5
    Alpha=1
    ConstraintWeights_0=0
    ConstraintWeights_1:float=0
    ConstraintWeights_2:float=0.0013
    ConstraintWeights_3:float=0.0838
    ConstraintWeights_4:float=0.0984
    ConstraintWeights_5:float=0.6381
    ConstraintWeights_6:float=0.0002
    ConstraintWeights_7:float=0.0003
    ConstraintWeights_8:float=0
    ConstraintWeights_9:float=0
    ConstraintWeights_10:float=0
    ConstraintWeights_11:float=0.1405
    ConstraintWeights_12:float=0
    ConstraintWeights_13:float=0
    ConstraintWeights_14:float=0
    ConstraintWeights_15:float=0.0152
    ConstraintWeights_16:float=0
    ConstraintWeights_17:float=0
    ConstraintWeights_18:float=0.0214
    ConstraintWeights_19:float=0.0007
    ConstraintWeights_20:float=0
   ) -> (Pose) {
    alphaBoolBlend {blendOption=Linear}
    boneToModify {boneName=dyn_lapel_upper_MSTR_R}
    constraintSetup [
     {targetBone {boneName=pelvis}}
     {targetBone {boneName=spine_01}}
     {targetBone {boneName=spine_02}}
     {targetBone {boneName=spine_03}}
     {targetBone {boneName=spine_04}}
     {targetBone {boneName=spine_05}}
     {targetBone {boneName=neck_01}}
     {targetBone {boneName=neck_02}}
     {targetBone {boneName=clavicle_l}}
     {targetBone {boneName=upperarm_twist_01_l}}
     {targetBone {boneName=upperarm_twist_02_l}}
     {targetBone {boneName=clavicle_r}}
     {targetBone {boneName=upperarm_twist_01_r}}
     {targetBone {boneName=upperarm_twist_02_r}}
     {targetBone {boneName=dyn_lapel_lower_01_L}}
     {targetBone {boneName=dyn_lapel_lower_01_R}}
     {targetBone {boneName=dyn_collar_01_L}}
     {targetBone {boneName=dyn_collar_02_L}}
     {targetBone {boneName=dyn_collar_01_R}}
     {targetBone {boneName=dyn_collar_02_R}}
     {targetBone {boneName=dyn_collar_bk_01}}
    ]
    constraintWeights [1,1,1,1,1,1,1,1,1,1,1,1,1,1,1,1,1,1,1,1,1]
   }
   n7 = "Constraint - dyn_lapel_lower_MSTR_L (PPPPPPPPPPPPPPPPPPP)"!(
    ComponentPose=@n6
    Alpha=1
    ConstraintWeights_0=0.0089
    ConstraintWeights_1:float=0.0001
    ConstraintWeights_2:float=0.0477
    ConstraintWeights_3:float=0.6534
    ConstraintWeights_4:float=0.1095
    ConstraintWeights_5:float=0.1281
    ConstraintWeights_6:float=0
    ConstraintWeights_7:float=0
    ConstraintWeights_8:float=0
    ConstraintWeights_9:float=0
    ConstraintWeights_10:float=0
    ConstraintWeights_11:float=0
    ConstraintWeights_12:float=0.0522
    ConstraintWeights_13:float=0
    ConstraintWeights_14:float=0.0001
    ConstraintWeights_15:float=0
    ConstraintWeights_16:float=0
    ConstraintWeights_17:float=0
    ConstraintWeights_18:float=0
   ) -> (Pose) {
    alphaBoolBlend {blendOption=Linear}
    boneToModify {boneName=dyn_lapel_lower_MSTR_L}
    constraintSetup [
     {targetBone {boneName=pelvis}}
     {targetBone {boneName=spine_01}}
     {targetBone {boneName=spine_02}}
     {targetBone {boneName=spine_03}}
     {targetBone {boneName=spine_04}}
     {targetBone {boneName=spine_05}}
     {targetBone {boneName=neck_01}}
     {targetBone {boneName=neck_02}}
     {targetBone {boneName=upperarm_twist_01_l}}
     {targetBone {boneName=upperarm_twist_02_l}}
     {targetBone {boneName=upperarm_twist_01_r}}
     {targetBone {boneName=upperarm_twist_02_r}}
     {targetBone {boneName=dyn_lapel_lower_01_L}}
     {targetBone {boneName=dyn_lapel_lower_01_R}}
     {targetBone {boneName=dyn_collar_01_L}}
     {targetBone {boneName=dyn_collar_02_L}}
     {targetBone {boneName=dyn_collar_01_R}}
     {targetBone {boneName=dyn_collar_02_R}}
     {targetBone {boneName=dyn_collar_bk_01}}
    ]
    constraintWeights [1,1,1,1,1,1,1,1,1,1,1,1,1,1,1,1,1,1,1]
   }
   "Constraint - dyn_lapel_lower_MSTR_R (PPPPPPPPPPPPPPPPPPPPP)"!(
    ComponentPose=@n7
    Alpha=1
    ConstraintWeights_0=0.0014
    ConstraintWeights_1:float=0
    ConstraintWeights_2:float=0.0395
    ConstraintWeights_3:float=0.6954
    ConstraintWeights_4:float=0.132
    ConstraintWeights_5:float=0.1099
    ConstraintWeights_6:float=0
    ConstraintWeights_7:float=0
    ConstraintWeights_8:float=0
    ConstraintWeights_9:float=0
    ConstraintWeights_10:float=0
    ConstraintWeights_11:float=0
    ConstraintWeights_12:float=0
    ConstraintWeights_13:float=0
    ConstraintWeights_14:float=0.0216
    ConstraintWeights_15:float=0
    ConstraintWeights_16:float=0
    ConstraintWeights_17:float=0.0002
    ConstraintWeights_18:float=0
    ConstraintWeights_19:float=0
    ConstraintWeights_20:float=0
   ) -> (Pose) {
    alphaBoolBlend {blendOption=Linear}
    boneToModify {boneName=dyn_lapel_lower_MSTR_R}
    constraintSetup [
     {targetBone {boneName=pelvis}}
     {targetBone {boneName=spine_01}}
     {targetBone {boneName=spine_02}}
     {targetBone {boneName=spine_03}}
     {targetBone {boneName=spine_04}}
     {targetBone {boneName=spine_05}}
     {targetBone {boneName=neck_01}}
     {targetBone {boneName=neck_02}}
     {targetBone {boneName=clavicle_l}}
     {targetBone {boneName=upperarm_twist_01_l}}
     {targetBone {boneName=upperarm_twist_02_l}}
     {targetBone {boneName=upperarm_twist_01_r}}
     {targetBone {boneName=upperarm_twist_02_r}}
     {targetBone {boneName=dyn_lapel_lower_01_L}}
     {targetBone {boneName=dyn_lapel_lower_01_R}}
     {targetBone {boneName=dyn_collar_01_L}}
     {targetBone {boneName=dyn_collar_02_L}}
     {targetBone {boneName=dyn_collar_01_R}}
     {targetBone {boneName=dyn_collar_02_R}}
     {targetBone {boneName=dyn_lapel_upper_01_L}}
     {targetBone {boneName=dyn_collar_bk_01}}
    ]
    constraintWeights [1,1,1,1,1,1,1,1,1,1,1,1,1,1,1,1,1,1,1,1,1]
   }
   "Transform (Modify) Bone - Bone: pistol_belt_R"!(ComponentPose, Alpha=1, Translation="0,0,-5.703", Rotation="0,0,0") -> (Pose) {
    rotationMode=BMM_Additive
    rotationSpace=BCS_BoneSpace
    translationMode=BMM_Additive
    translationSpace=BCS_BoneSpace
    alphaBoolBlend {blendOption=Linear}
    alphaScaleBiasClamp {interpSpeedDecreasing=5, interpSpeedIncreasing=5}
    boneToModify {boneName=pistol_belt_R}
    translation {z=-5.703}
   }
   n8 = "Transform (Modify) Bone - Bone: dyn_jacket_MSTR_L"!(ComponentPose, Alpha=1, Translation="0,0,0", Rotation="20,0,0") -> (Pose) {
    rotationMode=BMM_Additive
    rotationSpace=BCS_BoneSpace
    translationMode=BMM_Additive
    translationSpace=BCS_BoneSpace
    alphaBoolBlend {blendOption=Linear}
    boneToModify {boneName=dyn_jacket_MSTR_L}
    rotation {pitch=20}
   }
   n9 = "Transform (Modify) Bone - Bone: dyn_jacket_MSTR_L"(ComponentPose=@n8, Alpha=1, Translation="0,0,0", Rotation="29.790,-2.959,-11.021") {
    rotationMode=BMM_Additive
    rotationSpace=BCS_BoneSpace
    translationMode=BMM_Additive
    translationSpace=BCS_BoneSpace
    alphaBoolBlend {blendOption=Linear}
    boneToModify {boneName=dyn_jacket_MSTR_L}
    rotation=<rot 29.79 -2.959 -11.021>
   }
   n10 = "Transform (Modify) Bone - Bone: dyn_jacket_MSTR_R"!(ComponentPose=@n9, Alpha=1, Translation="0,0,0", Rotation="34.263,11.942,8.544") -> (Pose) {
    rotationMode=BMM_Additive
    rotationSpace=BCS_BoneSpace
    translationMode=BMM_Additive
    translationSpace=BCS_BoneSpace
    alphaBoolBlend {blendOption=Linear}
    boneToModify {boneName=dyn_jacket_MSTR_R}
    rotation=<rot 34.263 11.942 8.544>
   }
   "Transform (Modify) Bone - Bone: dyn_jacket_MSTR_R"(ComponentPose=@n10, Alpha=1, Translation="0,0,0", Rotation="0,0,0") {
    rotationMode=BMM_Additive
    rotationSpace=BCS_BoneSpace
    translationMode=BMM_Additive
    translationSpace=BCS_BoneSpace
    alphaBoolBlend {blendOption=Linear}
    boneToModify {boneName=dyn_jacket_MSTR_R}
   }
   n11 = Use cached pose 'main'()
   n12 = Use cached pose 'crouching'()
   n13 = Get bIsCrouching()
   n14 = RigidBody!(
    ComponentPose=Local To Component(LocalPose=Blend Poses by bool(BlendPose_0=@n12, BlendPose_1=@n11, bActiveValue=@n13) {blendType=Linear})
    Alpha
    OverrideWorldGravity:Vector="0,0,0"
    ComponentLinearAccScale:Vector="0.4,0.4,0.4"
    ComponentLinearVelScale:Vector="0.4,0.4,0.4"
    SimSpaceSettings:Sim Space Settings Structure=(WorldAlpha=0.5,VelocityScaleZ=0,DampingAlpha=1,MaxLinearVelocity=250,MaxAngularVelocity=3,MaxLinearAcceleration=350,MaxAngularAcceleration=500,ExternalLinearDragV=(X=0.1,Y=0.1,Z=0),ExternalLinearVelocity=(X=0,Y=0,Z=0),ExternalAngularVelocity=(X=0,Y=0,Z=0))
   ) -> (Pose) {
    bClampLinearTranslationLimitToRefPose=true
    bOverrideWorldGravity=true
    overridePhysicsAsset=&/M_MED_InfernoFNCS_Physics
    simulationSpace=BaseBoneSpace
    alphaBoolBlend {blendOption=Linear}
    baseBoneRef {boneName=pelvis}
    componentAppliedLinearAccClamp=<1000 1000 1000>
    componentLinearAccScale=<0.4 0.4 0.4>
    componentLinearVelScale=<0.4 0.4 0.4>
    overrideWorldGravity {z=-980}
    simSpaceSettings {
     maxAngularAcceleration=500
     maxAngularVelocity=3
     maxLinearAcceleration=350
     maxLinearVelocity=250
     velocityScaleZ=0
     worldAlpha=0.5
     externalLinearDragV=<0.1 0.1>
    }
   }
   n15 = Blend Poses by bool(BlendPose_0=Use cached pose 'sliding'(), BlendPose_1=Component To Local(ComponentPose=@n14), BlendTime_1=0.5, bActiveValue=Get bIsSliding()) {blendType=Linear}
   n16 = Use cached pose 'dbno'()
   n17 = Get bIsDBNO()
   n18 = Get bIsPlayingEmote()
   n19 = "Transform (Modify) Bone - Bone: dyn_lapel_upper_01_L"!(
    ComponentPose=Local To Component(LocalPose=Blend Poses by bool(BlendPose_0=@n16, BlendPose_1=@n15, bActiveValue=@n17) {blendType=Linear})
    bAlphaBoolEnabled:bool=@n18
    Translation="0,0,1"
    Rotation="7.264,12.351,1.584"
    Scale="1,1,1"
   ) -> (Pose) {
    alphaInputType=Bool
    rotationMode=BMM_Additive
    rotationSpace=BCS_BoneSpace
    translationMode=BMM_Additive
    translationSpace=BCS_BoneSpace
    alphaBoolBlend {blendInTime=0.2, blendOption=Linear, blendOutTime=0.2}
    boneToModify {boneName=dyn_lapel_upper_01_L}
    rotation=<rot 7.264 12.3509 1.5841>
    translation {z=1}
   }
   n20 = "Transform (Modify) Bone - Bone: dyn_lapel_upper_MSTR_R"!(ComponentPose=@n19, bAlphaBoolEnabled:bool=@n18, Translation="0,0,1", Rotation="7,0,0", Scale="1,1,1") -> (Pose) {
    alphaInputType=Bool
    rotationMode=BMM_Additive
    rotationSpace=BCS_BoneSpace
    translationMode=BMM_Additive
    translationSpace=BCS_BoneSpace
    alphaBoolBlend {blendInTime=0.2, blendOption=Linear, blendOutTime=0.2}
    boneToModify {boneName=dyn_lapel_upper_MSTR_R}
    rotation {pitch=7}
    translation {z=1}
   }
   Output Pose(Result=Component To Local(ComponentPose=@n20)) {layerGroup=None}
   n21 = "Constraint - dyn_pants_outer_02_MSTR_L (PPPP)"!(
    ComponentPose=Local To Component(LocalPose=Copy Pose From Mesh() {bUseAttachedParent=true})
    Alpha=1
    ConstraintWeights_0=0
    ConstraintWeights_1:float=0.9557
    ConstraintWeights_2:float=0.0012
    ConstraintWeights_3:float=0.043
   ) -> (Pose) {
    alphaBoolBlend {blendOption=Linear}
    boneToModify {boneName=dyn_pants_outer_02_MSTR_L}
    constraintSetup [
     {targetBone {boneName=calf_l}}
     {targetBone {boneName=calf_twist_01_l}}
     {targetBone {boneName=calf_twist_02_l}}
     {targetBone {boneName=foot_l}}
    ]
    constraintWeights [1,1,1,1]
   }
   n22 = "Constraint - dyn_pants_inner_02_MSTR_L (PPPPPP)"!(
    ComponentPose=@n21
    Alpha=1
    ConstraintWeights_0=0
    ConstraintWeights_1:float=0
    ConstraintWeights_2:float=0
    ConstraintWeights_3:float=0.8408
    ConstraintWeights_4:float=0.0015
    ConstraintWeights_5:float=0.1576
   ) -> (Pose) {
    alphaBoolBlend {blendOption=Linear}
    boneToModify {boneName=dyn_pants_inner_02_MSTR_L}
    constraintSetup [
     {targetBone {boneName=thigh_l}}
     {targetBone {boneName=thigh_twist_01_l}}
     {targetBone {boneName=calf_l}}
     {targetBone {boneName=calf_twist_01_l}}
     {targetBone {boneName=calf_twist_02_l}}
     {targetBone {boneName=foot_l}}
    ]
    constraintWeights [1,1,1,1,1,1]
   }
   n23 = "Constraint - dyn_pants_inner_01_MSTR_R (PPPPPP)"!(
    ComponentPose=@n22
    Alpha=1
    ConstraintWeights_0=0
    ConstraintWeights_1:float=0
    ConstraintWeights_2:float=0.0001
    ConstraintWeights_3:float=0.971
    ConstraintWeights_4:float=0.0276
    ConstraintWeights_5:float=0.0013
   ) -> (Pose) {
    alphaBoolBlend {blendOption=Linear}
    boneToModify {boneName=dyn_pants_inner_01_MSTR_R}
    constraintSetup [
     {targetBone {boneName=thigh_r}}
     {targetBone {boneName=thigh_twist_01_r}}
     {targetBone {boneName=calf_r}}
     {targetBone {boneName=calf_twist_01_r}}
     {targetBone {boneName=calf_twist_02_r}}
     {targetBone {boneName=foot_r}}
    ]
    constraintWeights [1,1,1,1,1,1]
   }
   n24 = "Constraint - dyn_pants_outer_01_MSTR_R (PPPPP)"!(
    ComponentPose=@n23
    Alpha=1
    ConstraintWeights_0=0
    ConstraintWeights_1:float=0
    ConstraintWeights_2:float=0.9867
    ConstraintWeights_3:float=0.0006
    ConstraintWeights_4:float=0.0127
   ) -> (Pose) {
    alphaBoolBlend {blendOption=Linear}
    boneToModify {boneName=dyn_pants_outer_01_MSTR_R}
    constraintSetup [
     {targetBone {boneName=thigh_r}}
     {targetBone {boneName=calf_r}}
     {targetBone {boneName=calf_twist_01_r}}
     {targetBone {boneName=calf_twist_02_r}}
     {targetBone {boneName=foot_r}}
    ]
    constraintWeights [1,1,1,1,1]
   }
   n25 = "Constraint - dyn_pants_outer_03_MSTR_L (PPPP)"!(
    ComponentPose=@n24
    Alpha=1
    ConstraintWeights_0=0
    ConstraintWeights_1:float=0.6896
    ConstraintWeights_2:float=0.0003
    ConstraintWeights_3:float=0.3101
   ) -> (Pose) {
    alphaBoolBlend {blendOption=Linear}
    boneToModify {boneName=dyn_pants_outer_03_MSTR_L}
    constraintSetup [
     {targetBone {boneName=calf_l}}
     {targetBone {boneName=calf_twist_01_l}}
     {targetBone {boneName=calf_twist_02_l}}
     {targetBone {boneName=foot_l}}
    ]
    constraintWeights [1,1,1,1]
   }
   n26 = "Constraint - dyn_pants_inner_03_L (PPPPPP)"!(
    ComponentPose=@n25
    Alpha=1
    ConstraintWeights_0=0
    ConstraintWeights_1:float=0
    ConstraintWeights_2:float=0
    ConstraintWeights_3:float=0.4602
    ConstraintWeights_4:float=0.0001
    ConstraintWeights_5:float=0.5397
   ) -> (Pose) {
    alphaBoolBlend {blendOption=Linear}
    boneToModify {boneName=dyn_pants_inner_03_L}
    constraintSetup [
     {targetBone {boneName=thigh_l}}
     {targetBone {boneName=thigh_twist_01_l}}
     {targetBone {boneName=calf_l}}
     {targetBone {boneName=calf_twist_01_l}}
     {targetBone {boneName=calf_twist_02_l}}
     {targetBone {boneName=foot_l}}
    ]
    constraintWeights [1,1,1,1,1,1]
   }
   n27 = "Constraint - dyn_pants_inner_02_R (PPPPPP)"!(
    ComponentPose=@n26
    Alpha=1
    ConstraintWeights_0=0
    ConstraintWeights_1:float=0
    ConstraintWeights_2:float=0
    ConstraintWeights_3:float=0.6849
    ConstraintWeights_4:float=0
    ConstraintWeights_5:float=0.315
   ) -> (Pose) {
    alphaBoolBlend {blendOption=Linear}
    boneToModify {boneName=dyn_pants_inner_02_R}
    constraintSetup [
     {targetBone {boneName=thigh_r}}
     {targetBone {boneName=thigh_twist_01_r}}
     {targetBone {boneName=calf_r}}
     {targetBone {boneName=calf_twist_01_r}}
     {targetBone {boneName=calf_twist_02_r}}
     {targetBone {boneName=foot_r}}
    ]
    constraintWeights [1,1,1,1,1,1]
   }
   n28 = "Constraint - dyn_pants_outer_02_MSTR_R (PPPPP)"!(
    ComponentPose=@n27
    Alpha=1
    ConstraintWeights_0=0
    ConstraintWeights_1:float=0
    ConstraintWeights_2:float=0.827
    ConstraintWeights_3:float=0.0001
    ConstraintWeights_4:float=0.1729
   ) -> (Pose) {
    alphaBoolBlend {blendOption=Linear}
    boneToModify {boneName=dyn_pants_outer_02_MSTR_R}
    constraintSetup [
     {targetBone {boneName=thigh_r}}
     {targetBone {boneName=calf_r}}
     {targetBone {boneName=calf_twist_01_r}}
     {targetBone {boneName=calf_twist_02_r}}
     {targetBone {boneName=foot_r}}
    ]
    constraintWeights [1,1,1,1,1]
   }
   n29 = "Constraint - dyn_fringe_01_MSTR_L (PPPPP)"!(
    ComponentPose=@n28
    Alpha=1
    ConstraintWeights_0=0.0003
    ConstraintWeights_1:float=0.0294
    ConstraintWeights_2:float=0.2747
    ConstraintWeights_3:float=0.4897
    ConstraintWeights_4:float=0.2059
   ) -> (Pose) {
    alphaBoolBlend {blendOption=Linear}
    boneToModify {boneName=dyn_fringe_01_MSTR_L}
    constraintSetup [
     {targetBone {boneName=spine_03}}
     {targetBone {boneName=spine_04}}
     {targetBone {boneName=spine_05}}
     {targetBone {boneName=clavicle_l}}
     {targetBone {boneName=upperarm_twist_01_l}}
    ]
    constraintWeights [1,1,1,1,1]
   }
   n30 = "Constraint - dyn_fringe_02_MSTR_L (PPPPP)"!(
    ComponentPose=@n29
    Alpha=1
    ConstraintWeights_0=0.0036
    ConstraintWeights_1:float=0.0376
    ConstraintWeights_2:float=0.285
    ConstraintWeights_3:float=0.4585
    ConstraintWeights_4:float=0.2153
   ) -> (Pose) {
    alphaBoolBlend {blendOption=Linear}
    boneToModify {boneName=dyn_fringe_02_MSTR_L}
    constraintSetup [
     {targetBone {boneName=spine_03}}
     {targetBone {boneName=spine_04}}
     {targetBone {boneName=spine_05}}
     {targetBone {boneName=clavicle_l}}
     {targetBone {boneName=upperarm_twist_01_l}}
    ]
    constraintWeights [1,1,1,1,1]
   }
   n31 = "Constraint - dyn_collar_MSTR_L (PP)"(ComponentPose=@n30, Alpha=1, ConstraintWeights_0=0.3, ConstraintWeights_1=0.7) {
    alphaBoolBlend {blendOption=Linear}
    boneToModify {boneName=dyn_collar_MSTR_L}
    constraintSetup [{targetBone {boneName=clavicle_l}}, {targetBone {boneName=spine_05}}]
    constraintWeights [1,0]
   }
   n32 = "Constraint - dyn_collar_MSTR_R (PP)"(ComponentPose=@n31, Alpha=1, ConstraintWeights_0=0.3, ConstraintWeights_1=0.7) {
    alphaBoolBlend {blendOption=Linear}
    boneToModify {boneName=dyn_collar_MSTR_R}
    constraintSetup [{targetBone {boneName=clavicle_r}}, {targetBone {boneName=spine_05}}]
    constraintWeights [1,0]
   }
   n33 = "Constraint - dyn_lapel_upper_MSTR_L (PP)"!(ComponentPose=@n32, Alpha=1, ConstraintWeights_0=0.4, ConstraintWeights_1:float=0.6) -> (Pose) {
    alphaBoolBlend {blendOption=Linear}
    boneToModify {boneName=dyn_lapel_upper_MSTR_L}
    constraintSetup [{targetBone {boneName=clavicle_l}}, {targetBone {boneName=spine_05}}]
    constraintWeights [1,0]
   }
   n34 = "Constraint - dyn_lapel_upper_MSTR_R (PP)"!(ComponentPose=@n33, Alpha=1, ConstraintWeights_0=0.4, ConstraintWeights_1:float=0.6) -> (Pose) {
    alphaBoolBlend {blendOption=Linear}
    boneToModify {boneName=dyn_lapel_upper_MSTR_R}
    constraintSetup [{targetBone {boneName=clavicle_r}}, {targetBone {boneName=spine_05}}]
    constraintWeights [1,0]
   }
   n35 = "Transform (Modify) Bone - Bone: dyn_jacket_MSTR_L"(ComponentPose=@n34, Alpha=1, Translation="0,0,0", Rotation="20,0,0") {
    rotationMode=BMM_Additive
    rotationSpace=BCS_BoneSpace
    translationMode=BMM_Additive
    translationSpace=BCS_BoneSpace
    alphaBoolBlend {blendOption=Linear}
    boneToModify {boneName=dyn_jacket_MSTR_L}
    rotation {pitch=20}
   }
   n36 = "Transform (Modify) Bone - Bone: dyn_jacket_MSTR_L"(ComponentPose=@n35, Alpha=1, Translation="0,0,0", Rotation="5.199,-20.281,-21.287") {
    rotationMode=BMM_Additive
    rotationSpace=BCS_BoneSpace
    translationMode=BMM_Additive
    translationSpace=BCS_BoneSpace
    alphaBoolBlend {blendOption=Linear}
    boneToModify {boneName=dyn_jacket_MSTR_L}
    rotation=<rot 5.199 -20.281 -21.2871>
   }
   n37 = "Transform (Modify) Bone - Bone: holster_MSTR"!(ComponentPose=@n36, Alpha=1, Rotation="10,0,0") -> (Pose) {
    rotationMode=BMM_Additive
    rotationSpace=BCS_BoneSpace
    translationSpace=BCS_BoneSpace
    alphaBoolBlend {blendOption=Linear}
    boneToModify {boneName=holster_MSTR}
    rotation {pitch=10}
   }
   n38 = "Transform (Modify) Bone - Bone: holster_MSTR"(ComponentPose=@n37, Alpha=1, Rotation="20,0,0") {
    rotationMode=BMM_Additive
    rotationSpace=BCS_BoneSpace
    translationSpace=BCS_BoneSpace
    alphaBoolBlend {blendOption=Linear}
    boneToModify {boneName=holster_MSTR}
    rotation {pitch=20}
   }
   n39 = "Transform (Modify) Bone - Bone: sleeve_L"!(ComponentPose=@n38, Alpha=1, Translation="-2.5,0,0") -> (Pose) {
    rotationSpace=BCS_BoneSpace
    translationMode=BMM_Additive
    translationSpace=BCS_BoneSpace
    alphaBoolBlend {blendOption=Linear}
    boneToModify {boneName=sleeve_L}
    translation {x=-2.5}
   }
   n40 = "Transform (Modify) Bone - Bone: sleeve_R"!(ComponentPose=@n39, Alpha=1, Translation="2.5,0,0") -> (Pose) {
    rotationSpace=BCS_BoneSpace
    translationMode=BMM_Additive
    translationSpace=BCS_BoneSpace
    alphaBoolBlend {blendOption=Linear}
    boneToModify {boneName=sleeve_R}
    translation {x=2.5}
   }
   n41 = "Transform (Modify) Bone - Bone: dyn_jacket_MSTR_R"(ComponentPose=@n40, Alpha=1, Translation="0,0,0", Rotation="14.489,10.035,5.274") {
    rotationMode=BMM_Additive
    rotationSpace=BCS_BoneSpace
    translationMode=BMM_Additive
    translationSpace=BCS_BoneSpace
    alphaBoolBlend {blendOption=Linear}
    boneToModify {boneName=dyn_jacket_MSTR_R}
    rotation=<rot 14.489 10.0349 5.2741>
   }
   n42 = "Transform (Modify) Bone - Bone: dyn_jacket_MSTR_R"(ComponentPose=@n41, Alpha=1, Translation="0,0,0", Rotation="10.000,-0.002,0.018") {
    rotationMode=BMM_Additive
    rotationSpace=BCS_BoneSpace
    translationMode=BMM_Additive
    translationSpace=BCS_BoneSpace
    alphaBoolBlend {blendOption=Linear}
    boneToModify {boneName=dyn_jacket_MSTR_R}
    rotation=<rot 10 -0.0021 0.0181>
   }
   n43 = "Transform (Modify) Bone - Bone: dyn_fringe_01_L"!(ComponentPose=@n42, Alpha=1, Rotation="-49.296,71.226,-56.942") -> (Pose) {
    rotationMode=BMM_Additive
    rotationSpace=BCS_BoneSpace
    translationSpace=BCS_BoneSpace
    alphaBoolBlend {blendOption=Linear}
    boneToModify {boneName=dyn_fringe_01_L}
    rotation=<rot -49.296 71.226 -56.942>
   }
   n44 = "Transform (Modify) Bone - Bone: dyn_fringe_02_L"!(ComponentPose=@n43, Alpha=1, Rotation="-60.000,-0.000,-0.000") -> (Pose) {
    rotationMode=BMM_Additive
    rotationSpace=BCS_BoneSpace
    translationSpace=BCS_BoneSpace
    alphaBoolBlend {blendOption=Linear}
    boneToModify {boneName=dyn_fringe_02_L}
    rotation=<rot -60 0 -0.0001>
   }
   n45 = "Transform (Modify) Bone - Bone: sleeve_L"(ComponentPose=@n44, Alpha=1, Translation="-1,0,0") {
    rotationSpace=BCS_BoneSpace
    translationMode=BMM_Additive
    translationSpace=BCS_BoneSpace
    alphaBoolBlend {blendOption=Linear}
    boneToModify {boneName=sleeve_L}
    translation {x=-1}
   }
   n46 = "Transform (Modify) Bone - Bone: sleeve_R"(ComponentPose=@n45, Alpha=1, Translation="1,0,0") {
    rotationSpace=BCS_BoneSpace
    translationMode=BMM_Additive
    translationSpace=BCS_BoneSpace
    alphaBoolBlend {blendOption=Linear}
    boneToModify {boneName=sleeve_R}
    translation {x=1}
   }
   n47 = Anim Dynamics - Bone: dyn_collar_01_L!(ComponentPose=@n46, Alpha=1, GravityOverride:Vector="0,0,-980") -> (Pose) {
    angularDampingOverride=1
    angularSpringConstant=20000
    bAngularSpring=true
    bOverrideAngularDamping=true
    bUseGravityOverride=true
    bUseSphericalLimits=true
    gravityScale=0.35
    simulationSpace=RootRelative
    alphaBoolBlend {blendOption=Linear}
    boundBone {boneName=dyn_collar_01_L}
    componentLinearAccScale=<0.7 0.7 0.7>
    componentLinearVelScale=<0.7 0.7 0.7>
    gravityOverride {z=-980}
    physicsBodyDefinitions [{
      collisionType=CustomSphere
      sphereCollisionRadius=2
      boundBone {boneName=dyn_collar_01_L}
      boxExtents=<25 35 35>
      constraintSetup {angularLimitsMax=<0 0 0>, angularLimitsMin=<0 -5 0>, angularTarget=<1 0 0>}
      localJointOffset=<4 0 0>
     }]
    sphericalLimits [{limitRadius=7, drivingBone {boneName=clavicle_l}, sphereLocalOffset=<9 0 0>}]
   }
   n48 = Anim Dynamics - Bone: dyn_collar_01_R!(ComponentPose=@n47, Alpha=1, GravityOverride:Vector="0,0,-980") -> (Pose) {
    angularDampingOverride=1
    angularSpringConstant=20000
    bAngularSpring=true
    bOverrideAngularDamping=true
    bUseGravityOverride=true
    bUseSphericalLimits=true
    gravityScale=0.35
    simulationSpace=RootRelative
    alphaBoolBlend {blendOption=Linear}
    boundBone {boneName=dyn_collar_01_R}
    componentLinearAccScale=<0.7 0.7 0.7>
    componentLinearVelScale=<0.7 0.7 0.7>
    gravityOverride {z=-980}
    physicsBodyDefinitions [{
      collisionType=CustomSphere
      sphereCollisionRadius=2
      boundBone {boneName=dyn_collar_01_R}
      boxExtents=<25 35 35>
      constraintSetup {angularLimitsMax=<0 0 0>, angularLimitsMin=<0 -5 0>, angularTarget=<1 0 0>}
      localJointOffset=<4 0 0>
     }]
    sphericalLimits [{limitRadius=7, drivingBone {boneName=clavicle_r}, sphereLocalOffset=<-9 0 0>}]
   }
   n49 = Anim Dynamics - Bone: dyn_collar_02_L!(ComponentPose=@n48, Alpha=1, GravityOverride:Vector="0,0,-980") -> (Pose) {
    angularDampingOverride=1
    angularSpringConstant=20000
    bAngularSpring=true
    bOverrideAngularDamping=true
    bUseGravityOverride=true
    bUseSphericalLimits=true
    gravityScale=0.35
    simulationSpace=RootRelative
    alphaBoolBlend {blendOption=Linear}
    boundBone {boneName=dyn_collar_02_L}
    componentLinearAccScale=<0.7 0.7 0.7>
    componentLinearVelScale=<0.7 0.7 0.7>
    gravityOverride {z=-980}
    physicsBodyDefinitions [{
      collisionType=CustomSphere
      sphereCollisionRadius=2
      boundBone {boneName=dyn_collar_02_L}
      boxExtents=<25 35 35>
      constraintSetup {angularLimitsMax=<0 0 0>, angularLimitsMin=<0 -5 0>, angularTarget=<1 0 0>}
      localJointOffset=<4 0 0>
     }]
    sphericalLimits [{limitRadius=7, drivingBone {boneName=clavicle_l}, sphereLocalOffset=<9 0 0>}]
   }
   n50 = Anim Dynamics - Bone: dyn_collar_02_R!(ComponentPose=@n49, Alpha=1, GravityOverride:Vector="0,0,-980") -> (Pose) {
    angularDampingOverride=1
    angularSpringConstant=20000
    bAngularSpring=true
    bOverrideAngularDamping=true
    bUseGravityOverride=true
    bUseSphericalLimits=true
    gravityScale=0.35
    simulationSpace=RootRelative
    alphaBoolBlend {blendOption=Linear}
    boundBone {boneName=dyn_collar_02_R}
    componentLinearAccScale=<0.7 0.7 0.7>
    componentLinearVelScale=<0.7 0.7 0.7>
    gravityOverride {z=-980}
    physicsBodyDefinitions [{
      collisionType=CustomSphere
      sphereCollisionRadius=2
      boundBone {boneName=dyn_collar_02_R}
      boxExtents=<25 35 35>
      constraintSetup {angularLimitsMax=<0 0 0>, angularLimitsMin=<0 -5 0>, angularTarget=<1 0 0>}
      localJointOffset=<4 0 0>
     }]
    sphericalLimits [{limitRadius=7, drivingBone {boneName=clavicle_r}, sphereLocalOffset=<-9 0 0>}]
   }
   n51 = Anim Dynamics - Bone: dyn_lapel_upper_01_L!(ComponentPose=@n50, Alpha=1, GravityOverride:Vector="0,0,-980") -> (Pose) {
    angularDampingOverride=1
    angularSpringConstant=20000
    bAngularSpring=true
    bOverrideAngularDamping=true
    bUseGravityOverride=true
    bUsePlanarLimit=false
    bUseSphericalLimits=true
    gravityScale=0.35
    simulationSpace=RootRelative
    alphaBoolBlend {blendOption=Linear}
    boundBone {boneName=dyn_lapel_upper_01_L}
    componentLinearAccScale=<0.7 0.7 0.7>
    componentLinearVelScale=<0.7 0.7 0.7>
    gravityOverride {z=-980}
    physicsBodyDefinitions [{
      collisionType=CustomSphere
      sphereCollisionRadius=1
      boundBone {boneName=dyn_lapel_upper_01_L}
      boxExtents=<25 35 35>
      constraintSetup {angularLimitsMax=<0 0 0>, angularLimitsMin=<0 -20 0>, angularTarget=<1 0 0>}
      localJointOffset=<4 0 0>
     }]
    sphericalLimits [{limitRadius=6, drivingBone {boneName=clavicle_l}, sphereLocalOffset=<5.2587 4.9923 -1.2266>}]
   }
   n52 = Anim Dynamics - Bone: dyn_lapel_upper_01_R!(ComponentPose=@n51, Alpha=1, GravityOverride:Vector="0,0,-980") -> (Pose) {
    angularDampingOverride=1
    angularSpringConstant=20000
    bAngularSpring=true
    bOverrideAngularDamping=true
    bUseGravityOverride=true
    bUsePlanarLimit=false
    bUseSphericalLimits=true
    gravityScale=0.35
    simulationSpace=RootRelative
    alphaBoolBlend {blendOption=Linear}
    boundBone {boneName=dyn_lapel_upper_01_R}
    componentLinearAccScale=<0.7 0.7 0.7>
    componentLinearVelScale=<0.7 0.7 0.7>
    gravityOverride {z=-980}
    physicsBodyDefinitions [{
      collisionType=CustomSphere
      sphereCollisionRadius=1
      boundBone {boneName=dyn_lapel_upper_01_R}
      boxExtents=<25 35 35>
      constraintSetup {angularLimitsMax=<0 0 0>, angularLimitsMin=<0 -25 0>, angularTarget=<1 0 0>}
      localJointOffset=<4 0 0>
     }]
    sphericalLimits [{limitRadius=6, drivingBone {boneName=clavicle_r}, sphereLocalOffset=<-5.7352 -5.0294 3.6593>}]
   }
   n53 = Anim Dynamics - Bone: dyn_lapel_lower_01_L!(ComponentPose=@n52, Alpha=1, GravityOverride:Vector="0,0,-980") -> (Pose) {
    angularDampingOverride=1
    angularSpringConstant=20000
    bAngularSpring=true
    bOverrideAngularDamping=true
    bUseGravityOverride=true
    bUseSphericalLimits=true
    gravityScale=0.35
    simulationSpace=RootRelative
    alphaBoolBlend {blendOption=Linear}
    boundBone {boneName=dyn_lapel_lower_01_L}
    componentLinearAccScale=<0.7 0.7 0.7>
    componentLinearVelScale=<0.7 0.7 0.7>
    gravityOverride {z=-980}
    physicsBodyDefinitions [{
      collisionType=CustomSphere
      sphereCollisionRadius=2
      boundBone {boneName=dyn_lapel_lower_01_L}
      boxExtents=<25 35 35>
      constraintSetup {angularLimitsMax=<0 0 0>, angularLimitsMin=<0 -25 0>, angularTarget=<1 0 0>}
      localJointOffset=<4 0 0>
     }]
    sphericalLimits [{limitRadius=7, drivingBone {boneName=clavicle_r}, sphereLocalOffset=<-9 0 0>}]
   }
   n54 = Anim Dynamics - Bone: dyn_lapel_lower_01_R!(ComponentPose=@n53, Alpha=1, GravityOverride:Vector="0,0,-980") -> (Pose) {
    angularDampingOverride=1
    angularSpringConstant=20000
    bAngularSpring=true
    bOverrideAngularDamping=true
    bUseGravityOverride=true
    bUseSphericalLimits=true
    gravityScale=0.35
    simulationSpace=RootRelative
    alphaBoolBlend {blendOption=Linear}
    boundBone {boneName=dyn_lapel_lower_01_R}
    componentLinearAccScale=<0.7 0.7 0.7>
    componentLinearVelScale=<0.7 0.7 0.7>
    gravityOverride {z=-980}
    physicsBodyDefinitions [{
      collisionType=CustomSphere
      sphereCollisionRadius=2
      boundBone {boneName=dyn_lapel_lower_01_R}
      boxExtents=<25 35 35>
      constraintSetup {angularLimitsMax=<0 0 0>, angularLimitsMin=<0 -25 0>, angularTarget=<1 0 0>}
      localJointOffset=<4 0 0>
     }]
    sphericalLimits [{limitRadius=7, drivingBone {boneName=clavicle_r}, sphereLocalOffset=<-9 0 0>}]
   }
   n55 = Anim Dynamics - Bone: holster!(ComponentPose=@n54, Alpha=1, GravityOverride:Vector="0,0,-980") -> (Pose) {
    angularDampingOverride=1
    angularSpringConstant=10000
    bAngularSpring=true
    bOverrideAngularDamping=true
    bUseGravityOverride=true
    gravityScale=0.35
    simulationSpace=RootRelative
    alphaBoolBlend {blendOption=Linear}
    boundBone {boneName=holster}
    componentAppliedLinearAccClamp=<100 100 100>
    componentLinearAccScale=<0.5 0.5 0.5>
    componentLinearVelScale=<0.5 0.5 0.5>
    gravityOverride {z=-980}
    physicsBodyDefinitions [{
      sphereCollisionRadius=2
      boundBone {boneName=holster}
      boxExtents=<40 40 40>
      constraintSetup {angularLimitsMax=<0 0 25>, angularLimitsMin=<0 -10 0>, angularTarget=<1 0 0>}
      localJointOffset=<5 0 0>
     }]
   }
   n56 = Anim Dynamics - Bone: dyn_pants_outer_01_L!(ComponentPose=@n55, Alpha=1, GravityOverride:Vector="0,0,0") -> (Pose) {
    angularDampingOverride=1
    angularSpringConstant=2000
    bAngularSpring=true
    bOverrideAngularDamping=true
    bUseGravityOverride=true
    bUsePlanarLimit=false
    gravityScale=0.35
    simulationSpace=RootRelative
    alphaBoolBlend {blendOption=Linear}
    boundBone {boneName=dyn_pants_outer_01_L}
    componentLinearAccScale=<0.7 0.7 0.7>
    componentLinearVelScale=<0.7 0.7 0.7>
    gravityOverride {z=-980}
    physicsBodyDefinitions [{
      sphereCollisionRadius=2
      boundBone {boneName=dyn_pants_outer_01_L}
      boxExtents=<17 17 17>
      constraintSetup {angularLimitsMax=<0 30 0>, angularLimitsMin=<0 -30 0>, angularTarget=<1 0 0>}
      localJointOffset=<4 0 0>
     }]
   }
   n57 = Anim Dynamics - Bone: dyn_pants_inner_01_L!(ComponentPose=@n56, Alpha=1, GravityOverride:Vector="0,0,0") -> (Pose) {
    angularDampingOverride=1
    angularSpringConstant=2000
    bAngularSpring=true
    bOverrideAngularDamping=true
    bUseGravityOverride=true
    bUsePlanarLimit=false
    gravityScale=0.35
    simulationSpace=RootRelative
    alphaBoolBlend {blendOption=Linear}
    boundBone {boneName=dyn_pants_inner_01_L}
    componentLinearAccScale=<0.7 0.7 0.7>
    componentLinearVelScale=<0.7 0.7 0.7>
    gravityOverride {z=-980}
    physicsBodyDefinitions [{
      sphereCollisionRadius=2
      boundBone {boneName=dyn_pants_inner_01_L}
      boxExtents=<20 20 20>
      constraintSetup {angularLimitsMax=<0 30 0>, angularLimitsMin=<0 -30 0>, angularTarget=<1 0 0>}
      localJointOffset=<4 0 0>
     }]
   }
   n58 = Anim Dynamics - Bone: dyn_pants_outer_02_L!(ComponentPose=@n57, Alpha=1, GravityOverride:Vector="0,0,0") -> (Pose) {
    angularDampingOverride=1
    angularSpringConstant=2000
    bAngularSpring=true
    bOverrideAngularDamping=true
    bUseGravityOverride=true
    bUsePlanarLimit=false
    gravityScale=0.35
    simulationSpace=RootRelative
    alphaBoolBlend {blendOption=Linear}
    boundBone {boneName=dyn_pants_outer_02_L}
    componentLinearAccScale=<0.7 0.7 0.7>
    componentLinearVelScale=<0.7 0.7 0.7>
    gravityOverride {z=-980}
    physicsBodyDefinitions [{
      sphereCollisionRadius=2
      boundBone {boneName=dyn_pants_outer_02_L}
      boxExtents=<15 15 15>
      constraintSetup {angularLimitsMax=<0 30 0>, angularLimitsMin=<0 -30 0>, angularTarget=<1 0 0>}
      localJointOffset=<4 0 0>
     }]
   }
   n59 = Anim Dynamics - Bone: dyn_pants_inner_02_L!(ComponentPose=@n58, Alpha=1, GravityOverride:Vector="0,0,0") -> (Pose) {
    angularDampingOverride=1
    angularSpringConstant=2000
    bAngularSpring=true
    bOverrideAngularDamping=true
    bUseGravityOverride=true
    bUsePlanarLimit=false
    gravityScale=0.35
    simulationSpace=RootRelative
    alphaBoolBlend {blendOption=Linear}
    boundBone {boneName=dyn_pants_inner_02_L}
    componentLinearAccScale=<0.7 0.7 0.7>
    componentLinearVelScale=<0.7 0.7 0.7>
    gravityOverride {z=-980}
    physicsBodyDefinitions [{
      sphereCollisionRadius=2
      boundBone {boneName=dyn_pants_inner_02_L}
      boxExtents=<16 16 16>
      constraintSetup {angularLimitsMax=<0 30 0>, angularLimitsMin=<0 -30 0>, angularTarget=<1 0 0>}
      localJointOffset=<4 0 0>
     }]
   }
   n60 = Anim Dynamics - Bone: dyn_pants_outer_03_L!(ComponentPose=@n59, Alpha=1, GravityOverride:Vector="0,0,0") -> (Pose) {
    angularDampingOverride=1
    angularSpringConstant=2000
    bAngularSpring=true
    bOverrideAngularDamping=true
    bUseGravityOverride=true
    bUsePlanarLimit=false
    gravityScale=0.35
    simulationSpace=RootRelative
    alphaBoolBlend {blendOption=Linear}
    boundBone {boneName=dyn_pants_outer_03_L}
    componentLinearAccScale=<0.7 0.7 0.7>
    componentLinearVelScale=<0.7 0.7 0.7>
    gravityOverride {z=-980}
    physicsBodyDefinitions [{
      sphereCollisionRadius=2
      boundBone {boneName=dyn_pants_outer_03_L}
      boxExtents=<18 18 18>
      constraintSetup {angularLimitsMax=<0 30 0>, angularLimitsMin=<0 -30 0>, angularTarget=<1 0 0>}
      localJointOffset=<4 0 0>
     }]
   }
   n61 = Anim Dynamics - Bone: dyn_pants_inner_03_L!(ComponentPose=@n60, Alpha=1, GravityOverride:Vector="0,0,0") -> (Pose) {
    angularDampingOverride=1
    angularSpringConstant=2000
    bAngularSpring=true
    bOverrideAngularDamping=true
    bUseGravityOverride=true
    bUsePlanarLimit=false
    gravityScale=0.35
    simulationSpace=RootRelative
    alphaBoolBlend {blendOption=Linear}
    boundBone {boneName=dyn_pants_inner_03_L}
    componentLinearAccScale=<0.7 0.7 0.7>
    componentLinearVelScale=<0.7 0.7 0.7>
    gravityOverride {z=-980}
    physicsBodyDefinitions [{
      sphereCollisionRadius=2
      boundBone {boneName=dyn_pants_inner_03_L}
      boxExtents=<18 18 18>
      constraintSetup {angularLimitsMax=<0 30 0>, angularLimitsMin=<0 -30 0>, angularTarget=<1 0 0>}
      localJointOffset=<4 0 0>
     }]
   }
   n62 = Anim Dynamics - Bone: dyn_pants_outer_04_L!(ComponentPose=@n61, Alpha=1, GravityOverride:Vector="0,0,0") -> (Pose) {
    angularDampingOverride=1
    angularSpringConstant=2000
    bAngularSpring=true
    bOverrideAngularDamping=true
    bUseGravityOverride=true
    bUsePlanarLimit=false
    gravityScale=0.35
    simulationSpace=RootRelative
    alphaBoolBlend {blendOption=Linear}
    boundBone {boneName=dyn_pants_outer_04_L}
    componentLinearAccScale=<0.7 0.7 0.7>
    componentLinearVelScale=<0.7 0.7 0.7>
    gravityOverride {z=-980}
    physicsBodyDefinitions [{
      sphereCollisionRadius=2
      boundBone {boneName=dyn_pants_outer_04_L}
      boxExtents=<20 20 20>
      constraintSetup {angularLimitsMax=<0 0 0>, angularLimitsMin=<0 -30 0>, angularTarget=<1 0 0>}
      localJointOffset=<4 0 0>
     }]
   }
   n63 = Anim Dynamics - Bone: dyn_pants_inner_04_L!(ComponentPose=@n62, Alpha=1, GravityOverride:Vector="0,0,0") -> (Pose) {
    angularDampingOverride=1
    angularSpringConstant=2000
    bAngularSpring=true
    bOverrideAngularDamping=true
    bUseGravityOverride=true
    bUsePlanarLimit=false
    gravityScale=0.35
    simulationSpace=RootRelative
    alphaBoolBlend {blendOption=Linear}
    boundBone {boneName=dyn_pants_inner_04_L}
    componentLinearAccScale=<0.7 0.7 0.7>
    componentLinearVelScale=<0.7 0.7 0.7>
    gravityOverride {z=-980}
    physicsBodyDefinitions [{
      sphereCollisionRadius=2
      boundBone {boneName=dyn_pants_inner_04_L}
      boxExtents=<10 10 10>
      constraintSetup {angularLimitsMax=<0 0 0>, angularLimitsMin=<0 -30 0>, angularTarget=<1 0 0>}
      localJointOffset=<4 0 0>
     }]
   }
   n64 = Anim Dynamics - Bone: dyn_pants_inner_01_R!(ComponentPose=@n63, Alpha=1, GravityOverride:Vector="0,0,0") -> (Pose) {
    angularDampingOverride=1
    angularSpringConstant=2000
    bAngularSpring=true
    bOverrideAngularDamping=true
    bUseGravityOverride=true
    bUsePlanarLimit=false
    gravityScale=0.35
    simulationSpace=RootRelative
    alphaBoolBlend {blendOption=Linear}
    boundBone {boneName=dyn_pants_inner_01_R}
    componentLinearAccScale=<0.7 0.7 0.7>
    componentLinearVelScale=<0.7 0.7 0.7>
    gravityOverride {z=-980}
    physicsBodyDefinitions [{
      sphereCollisionRadius=2
      boundBone {boneName=dyn_pants_inner_01_R}
      boxExtents=<16 16 16>
      constraintSetup {angularLimitsMax=<0 30 0>, angularLimitsMin=<0 -30 0>, angularTarget=<1 0 0>}
      localJointOffset=<4 0 0>
     }]
   }
   n65 = Anim Dynamics - Bone: dyn_pants_outer_01_R!(ComponentPose=@n64, Alpha=1, GravityOverride:Vector="0,0,0") -> (Pose) {
    angularDampingOverride=1
    angularSpringConstant=2000
    bAngularSpring=true
    bOverrideAngularDamping=true
    bUseGravityOverride=true
    bUsePlanarLimit=false
    gravityScale=0.35
    simulationSpace=RootRelative
    alphaBoolBlend {blendOption=Linear}
    boundBone {boneName=dyn_pants_outer_01_R}
    componentLinearAccScale=<0.7 0.7 0.7>
    componentLinearVelScale=<0.7 0.7 0.7>
    gravityOverride {z=-980}
    physicsBodyDefinitions [{
      sphereCollisionRadius=2
      boundBone {boneName=dyn_pants_outer_01_R}
      boxExtents=<15 15 15>
      constraintSetup {angularLimitsMax=<0 30 0>, angularLimitsMin=<0 -30 0>, angularTarget=<1 0 0>}
      localJointOffset=<4 0 0>
     }]
   }
   n66 = Anim Dynamics - Bone: dyn_pants_inner_02_R!(ComponentPose=@n65, Alpha=1, GravityOverride:Vector="0,0,0") -> (Pose) {
    angularDampingOverride=1
    angularSpringConstant=2000
    bAngularSpring=true
    bOverrideAngularDamping=true
    bUseGravityOverride=true
    bUsePlanarLimit=false
    gravityScale=0.35
    simulationSpace=RootRelative
    alphaBoolBlend {blendOption=Linear}
    boundBone {boneName=dyn_pants_inner_02_R}
    componentLinearAccScale=<0.7 0.7 0.7>
    componentLinearVelScale=<0.7 0.7 0.7>
    gravityOverride {z=-980}
    physicsBodyDefinitions [{
      sphereCollisionRadius=2
      boundBone {boneName=dyn_pants_inner_02_R}
      boxExtents=<20 20 20>
      constraintSetup {angularLimitsMax=<0 30 0>, angularLimitsMin=<0 -30 0>, angularTarget=<1 0 0>}
      localJointOffset=<4 0 0>
     }]
   }
   n67 = Anim Dynamics - Bone: dyn_pants_outer_02_R!(ComponentPose=@n66, Alpha=1, GravityOverride:Vector="0,0,0") -> (Pose) {
    angularDampingOverride=1
    angularSpringConstant=2000
    bAngularSpring=true
    bOverrideAngularDamping=true
    bUseGravityOverride=true
    bUsePlanarLimit=false
    gravityScale=0.35
    simulationSpace=RootRelative
    alphaBoolBlend {blendOption=Linear}
    boundBone {boneName=dyn_pants_outer_02_R}
    componentLinearAccScale=<0.7 0.7 0.7>
    componentLinearVelScale=<0.7 0.7 0.7>
    gravityOverride {z=-980}
    physicsBodyDefinitions [{
      sphereCollisionRadius=2
      boundBone {boneName=dyn_pants_outer_02_R}
      boxExtents=<17 17 17>
      constraintSetup {angularLimitsMax=<0 0 0>, angularLimitsMin=<0 -30 0>, angularTarget=<1 0 0>}
      localJointOffset=<4 0 0>
     }]
   }
   n68 = Anim Dynamics - Bone: dyn_pants_inner_03_R!(ComponentPose=@n67, Alpha=1, GravityOverride:Vector="0,0,0") -> (Pose) {
    angularDampingOverride=1
    angularSpringConstant=2000
    bAngularSpring=true
    bOverrideAngularDamping=true
    bUseGravityOverride=true
    bUsePlanarLimit=false
    gravityScale=0.35
    simulationSpace=RootRelative
    alphaBoolBlend {blendOption=Linear}
    boundBone {boneName=dyn_pants_inner_03_R}
    componentLinearAccScale=<0.7 0.7 0.7>
    componentLinearVelScale=<0.7 0.7 0.7>
    gravityOverride {z=-980}
    physicsBodyDefinitions [{
      sphereCollisionRadius=2
      boundBone {boneName=dyn_pants_inner_03_R}
      boxExtents=<15 15 15>
      constraintSetup {angularLimitsMax=<0 0 0>, angularLimitsMin=<0 -30 0>, angularTarget=<1 0 0>}
      localJointOffset=<4 0 0>
     }]
   }
   n69 = Anim Dynamics - Bone: dyn_pants_outer_03_R!(ComponentPose=@n68, Alpha=1, GravityOverride:Vector="0,0,0") -> (Pose) {
    angularDampingOverride=1
    angularSpringConstant=2000
    bAngularSpring=true
    bOverrideAngularDamping=true
    bUseGravityOverride=true
    bUsePlanarLimit=false
    gravityScale=0.35
    simulationSpace=RootRelative
    alphaBoolBlend {blendOption=Linear}
    boundBone {boneName=dyn_pants_outer_03_R}
    componentLinearAccScale=<0.7 0.7 0.7>
    componentLinearVelScale=<0.7 0.7 0.7>
    gravityOverride {z=-980}
    physicsBodyDefinitions [{
      sphereCollisionRadius=2
      boundBone {boneName=dyn_pants_outer_03_R}
      boxExtents=<16 16 16>
      constraintSetup {angularLimitsMax=<0 0 0>, angularLimitsMin=<0 -30 0>, angularTarget=<1 0 0>}
      localJointOffset=<4 0 0>
     }]
   }
   n70 = Anim Dynamics - Bone: dyn_pants_inner_04_R!(ComponentPose=@n69, Alpha=1, GravityOverride:Vector="0,0,0") -> (Pose) {
    angularDampingOverride=1
    angularSpringConstant=2000
    bAngularSpring=true
    bOverrideAngularDamping=true
    bUseGravityOverride=true
    bUsePlanarLimit=false
    gravityScale=0.35
    simulationSpace=RootRelative
    alphaBoolBlend {blendOption=Linear}
    boundBone {boneName=dyn_pants_inner_04_R}
    componentLinearAccScale=<0.7 0.7 0.7>
    componentLinearVelScale=<0.7 0.7 0.7>
    gravityOverride {z=-980}
    physicsBodyDefinitions [{
      sphereCollisionRadius=2
      boundBone {boneName=dyn_pants_inner_04_R}
      boxExtents=<10 10 10>
      constraintSetup {angularLimitsMax=<0 0 0>, angularLimitsMin=<0 -30 0>, angularTarget=<1 0 0>}
      localJointOffset=<4 0 0>
     }]
   }
   Save cached pose 'start pose'(Pose=Component To Local(ComponentPose=@n70)) {cacheName=start pose}
   n71 = "Transform (Modify) Bone - Bone: pistol_belt_L"!(ComponentPose=Local To Component(LocalPose=Use cached pose 'start pose'()), Alpha=1, Translation="7,-1,0", Rotation="0,0,0") -> (Pose) {
    rotationMode=BMM_Additive
    rotationSpace=BCS_BoneSpace
    translationMode=BMM_Additive
    translationSpace=BCS_BoneSpace
    alphaBoolBlend {blendOption=Linear}
    boneToModify {boneName=pistol_belt_L}
    translation=<7 -1>
   }
   n72 = "Transform (Modify) Bone - Bone: pistol_belt_L"(ComponentPose=@n71, Alpha=1, Translation="0,0,1.5", Rotation="0,0,0") {
    rotationMode=BMM_Additive
    rotationSpace=BCS_BoneSpace
    translationMode=BMM_Additive
    translationSpace=BCS_BoneSpace
    alphaBoolBlend {blendOption=Linear}
    boneToModify {boneName=pistol_belt_L}
    translation {z=1.5}
   }
   n73 = "Transform (Modify) Bone - Bone: pistol_belt_R"(ComponentPose=@n72, Alpha=1, Translation="12.464,1.987,0", Rotation="0,0,0") {
    rotationMode=BMM_Additive
    rotationSpace=BCS_BoneSpace
    translationMode=BMM_Additive
    translationSpace=BCS_BoneSpace
    alphaBoolBlend {blendOption=Linear}
    boneToModify {boneName=pistol_belt_R}
    translation=<12.464 1.987>
   }
   n74 = "Transform (Modify) Bone - Bone: pistol_belt_R"(ComponentPose=@n73, Alpha=1, Translation="0,-1,-4", Rotation="0,0,0") {
    rotationMode=BMM_Additive
    rotationSpace=BCS_BoneSpace
    translationMode=BMM_Additive
    translationSpace=BCS_BoneSpace
    alphaBoolBlend {blendOption=Linear}
    alphaScaleBiasClamp {interpSpeedDecreasing=20, interpSpeedIncreasing=20}
    boneToModify {boneName=pistol_belt_R}
    translation {y=-1, z=-4}
   }
   n75 = "Transform (Modify) Bone - Bone: pistol_belt_R"(ComponentPose=@n74, Alpha=1, Translation="0,0,3", Rotation="0,0,0") {
    translationMode=BMM_Additive
    translationSpace=BCS_BoneSpace
    alphaBoolBlend {blendOption=Linear}
    boneToModify {boneName=pistol_belt_R}
    translation {z=3}
   }
   n76 = "Transform (Modify) Bone - Bone: pistol_belt_L"(ComponentPose=Local To Component(LocalPose=Use cached pose 'start pose'()), Alpha=1, Translation="4.5,0,0", Rotation="0,0,0") {
    rotationMode=BMM_Additive
    rotationSpace=BCS_BoneSpace
    translationMode=BMM_Additive
    translationSpace=BCS_BoneSpace
    alphaBoolBlend {blendOption=Linear}
    boneToModify {boneName=pistol_belt_L}
    translation {x=4.5}
   }
   n77 = "Transform (Modify) Bone - Bone: pistol_belt_R"(ComponentPose=@n76, Alpha=1, Translation="9.5,0,0", Rotation="0,0,0") {
    rotationMode=BMM_Additive
    rotationSpace=BCS_BoneSpace
    translationMode=BMM_Additive
    translationSpace=BCS_BoneSpace
    alphaBoolBlend {blendOption=Linear}
    boneToModify {boneName=pistol_belt_R}
    translation {x=9.5}
   }
   Save cached pose 'main'(Pose=Blend Poses by bool(
     BlendPose_0=Component To Local(ComponentPose=@n77)
     BlendPose_1=Component To Local(ComponentPose=@n75)
     BlendTime_0=1
     BlendTime_1=0.2
     bActiveValue=Get bIsSprinting()
    ) {blendType=Linear}) {cacheName=main}
   n78 = "Transform (Modify) Bone - Bone: pistol_belt_L"(ComponentPose=Local To Component(LocalPose=Use cached pose 'start pose'()), Alpha=1, Translation="7.000,-0.562,0.048", Rotation="0,0,0") {
    translationMode=BMM_Additive
    translationSpace=BCS_BoneSpace
    alphaBoolBlend {blendOption=Linear}
    boneToModify {boneName=pistol_belt_L}
    translation=<7 -0.562 0.048>
   }
   n79 = "Transform (Modify) Bone - Bone: pistol_belt_R"(ComponentPose=@n78, Alpha=1, Translation="8.000,1.315,-0.010", Rotation="0,0,0") {
    translationMode=BMM_Additive
    translationSpace=BCS_BoneSpace
    alphaBoolBlend {blendOption=Linear}
    boneToModify {boneName=pistol_belt_R}
    translation=<8 1.315 -0.01>
   }
   Save cached pose 'crouching'(Pose=Component To Local(ComponentPose=@n79)) {cacheName=crouching}
   n80 = "Transform (Modify) Bone - Bone: pistol_belt_L"(ComponentPose=Local To Component(LocalPose=Use cached pose 'start pose'()), Alpha=1, Translation="8,-0.562,0.048", Rotation="0,0,0") {
    translationMode=BMM_Additive
    translationSpace=BCS_BoneSpace
    alphaBoolBlend {blendOption=Linear}
    boneToModify {boneName=pistol_belt_L}
    translation=<8 -0.562 0.048>
   }
   n81 = "Transform (Modify) Bone - Bone: pistol_belt_R"(ComponentPose=@n80, Alpha=1, Translation="12.000,2.491,-0.154", Rotation="0,0,0") {
    translationMode=BMM_Additive
    translationSpace=BCS_BoneSpace
    alphaBoolBlend {blendOption=Linear}
    boneToModify {boneName=pistol_belt_R}
    translation=<12 2.491 -0.154>
   }
   n82 = "Transform (Modify) Bone - Bone: dyn_jacket_MSTR_L"(ComponentPose=@n81, Alpha=1, Translation="0,0,0", Rotation="20,0,0") {
    rotationMode=BMM_Additive
    rotationSpace=BCS_BoneSpace
    translationMode=BMM_Additive
    translationSpace=BCS_BoneSpace
    alphaBoolBlend {blendOption=Linear}
    boneToModify {boneName=dyn_jacket_MSTR_L}
    rotation {pitch=20}
   }
   n83 = "Transform (Modify) Bone - Bone: dyn_jacket_MSTR_R"(ComponentPose=@n82, Alpha=1, Translation="0,0,0", Rotation="0,-35,0") {
    rotationMode=BMM_Additive
    rotationSpace=BCS_BoneSpace
    translationMode=BMM_Additive
    translationSpace=BCS_BoneSpace
    alphaBoolBlend {blendOption=Linear}
    boneToModify {boneName=dyn_jacket_MSTR_R}
    rotation {yaw=-35}
   }
   n84 = RigidBody(
    ComponentPose=@n83
    OverrideWorldGravity="0,0,0"
    ComponentLinearAccScale="0.4,0.4,0.4"
    ComponentLinearVelScale="0.4,0.4,0.4"
    SimSpaceSettings=(WorldAlpha=0.5,VelocityScaleZ=0,DampingAlpha=1,MaxLinearVelocity=250,MaxAngularVelocity=3,MaxLinearAcceleration=350,MaxAngularAcceleration=500,ExternalLinearDragV=(X=0.1,Y=0.1,Z=0),ExternalLinearVelocity=(X=0,Y=0,Z=0),ExternalAngularVelocity=(X=0,Y=0,Z=0))
   ) {
    bClampLinearTranslationLimitToRefPose=true
    bOverrideWorldGravity=true
    overridePhysicsAsset=&/M_MED_InfernoFNCS_DBNO_Physics
    simulationSpace=BaseBoneSpace
    alphaBoolBlend {blendOption=Linear}
    baseBoneRef {boneName=pelvis}
    componentAppliedLinearAccClamp=<1000 1000 1000>
    componentLinearAccScale=<0.4 0.4 0.4>
    componentLinearVelScale=<0.4 0.4 0.4>
    overrideWorldGravity {z=-980}
    simSpaceSettings {
     maxAngularAcceleration=500
     maxAngularVelocity=3
     maxLinearAcceleration=350
     maxLinearVelocity=250
     velocityScaleZ=0
     worldAlpha=0.5
     externalLinearDragV=<0.1 0.1>
    }
   }
   n85 = "Transform (Modify) Bone - Bone: dyn_jacket_02_L"!(ComponentPose=@n84, Alpha=1, Translation="0,0,0", Rotation="10,20,0") -> (Pose) {
    rotationMode=BMM_Additive
    rotationSpace=BCS_BoneSpace
    translationMode=BMM_Additive
    translationSpace=BCS_BoneSpace
    alphaBoolBlend {blendOption=Linear}
    boneToModify {boneName=dyn_jacket_02_L}
    rotation {pitch=10, yaw=20}
   }
   n86 = "Transform (Modify) Bone - Bone: dyn_jacket_04_L"!(ComponentPose=@n85, Alpha=1, Translation="0,0,0", Rotation="30,0,0") -> (Pose) {
    rotationMode=BMM_Additive
    rotationSpace=BCS_BoneSpace
    translationMode=BMM_Additive
    translationSpace=BCS_BoneSpace
    alphaBoolBlend {blendOption=Linear}
    boneToModify {boneName=dyn_jacket_04_L}
    rotation {pitch=30}
   }
   n87 = "Transform (Modify) Bone - Bone: dyn_jacket_02_R"!(ComponentPose=@n86, Alpha=1, Translation="0,0,0", Rotation="10,-50,0") -> (Pose) {
    rotationMode=BMM_Additive
    rotationSpace=BCS_BoneSpace
    translationMode=BMM_Additive
    translationSpace=BCS_BoneSpace
    alphaBoolBlend {blendOption=Linear}
    boneToModify {boneName=dyn_jacket_02_R}
    rotation {pitch=10, yaw=-50}
   }
   n88 = "Transform (Modify) Bone - Bone: dyn_jacket_03_R"!(ComponentPose=@n87, Alpha=1, Translation="0,0,0", Rotation="0,10,0") -> (Pose) {
    rotationMode=BMM_Additive
    rotationSpace=BCS_BoneSpace
    translationMode=BMM_Additive
    translationSpace=BCS_BoneSpace
    alphaBoolBlend {blendOption=Linear}
    boneToModify {boneName=dyn_jacket_03_R}
    rotation {yaw=10}
   }
   Save cached pose 'dbno'(Pose=Component To Local(ComponentPose=@n88)) {cacheName=dbno}
   n89 = "Transform (Modify) Bone - Bone: holster_MSTR"(ComponentPose=Local To Component(LocalPose=Use cached pose 'start pose'()), Alpha=1, Rotation="40,40,0") {
    rotationMode=BMM_Additive
    rotationSpace=BCS_BoneSpace
    translationSpace=BCS_BoneSpace
    alphaBoolBlend {blendOption=Linear}
    boneToModify {boneName=holster_MSTR}
    rotation {pitch=40, yaw=40}
   }
   n90 = "Transform (Modify) Bone - Bone: pistol_belt_L"(ComponentPose=@n89, Alpha=1, Translation="8,-0.562,0.048", Rotation="0,0,0") {
    translationMode=BMM_Additive
    translationSpace=BCS_BoneSpace
    alphaBoolBlend {blendOption=Linear}
    boneToModify {boneName=pistol_belt_L}
    translation=<8 -0.562 0.048>
   }
   n91 = "Transform (Modify) Bone - Bone: pistol_belt_R"(ComponentPose=@n90, Alpha=1, Translation="14,1.315,-0.01", Rotation="0,0,0") {
    translationMode=BMM_Additive
    translationSpace=BCS_BoneSpace
    alphaBoolBlend {blendOption=Linear}
    boneToModify {boneName=pistol_belt_R}
    translation=<14 1.315 -0.01>
   }
   n92 = "Transform (Modify) Bone - Bone: dyn_jacket_bk_01"!(ComponentPose=@n91, Alpha=1, Translation="0,0,0", Rotation="60.000,-180.000,180.000") -> (Pose) {
    rotationMode=BMM_Additive
    rotationSpace=BCS_BoneSpace
    translationMode=BMM_Additive
    translationSpace=BCS_BoneSpace
    alphaBoolBlend {blendOption=Linear}
    boneToModify {boneName=dyn_jacket_bk_01}
    rotation=<rot 60.0001 -180 180>
   }
   n93 = "Transform (Modify) Bone - Bone: dyn_jacket_mid_01_L"!(ComponentPose=@n92, Alpha=1, Translation="0,0,0", Rotation="36.822,-61.628,-36.967") -> (Pose) {
    rotationMode=BMM_Additive
    rotationSpace=BCS_BoneSpace
    translationMode=BMM_Additive
    translationSpace=BCS_BoneSpace
    alphaBoolBlend {blendOption=Linear}
    boneToModify {boneName=dyn_jacket_mid_01_L}
    rotation=<rot 36.822 -61.6279 -36.9669>
   }
   n94 = "Transform (Modify) Bone - Bone: dyn_jacket_mid_01_R"!(ComponentPose=@n93, Alpha=1, Translation="0,0,0", Rotation="74.960,91.494,84.309") -> (Pose) {
    rotationMode=BMM_Additive
    rotationSpace=BCS_BoneSpace
    translationMode=BMM_Additive
    translationSpace=BCS_BoneSpace
    alphaBoolBlend {blendOption=Linear}
    boneToModify {boneName=dyn_jacket_mid_01_R}
    rotation=<rot 74.96 91.4938 84.3088>
   }
   n95 = "Transform (Modify) Bone - Bone: dyn_jacket_MSTR_L"(ComponentPose=@n94, Alpha=1, Translation="0,0,0", Rotation="-0.111,-38.804,-10.741") {
    rotationMode=BMM_Additive
    rotationSpace=BCS_BoneSpace
    translationMode=BMM_Additive
    translationSpace=BCS_BoneSpace
    alphaBoolBlend {blendOption=Linear}
    boneToModify {boneName=dyn_jacket_MSTR_L}
    rotation=<rot -0.111 -38.8039 -10.7411>
   }
   n96 = "Transform (Modify) Bone - Bone: dyn_jacket_MSTR_R"(ComponentPose=@n95, Alpha=1, Translation="0,0,0", Rotation="6.409,50.431,7.692") {
    rotationMode=BMM_Additive
    rotationSpace=BCS_BoneSpace
    translationMode=BMM_Additive
    translationSpace=BCS_BoneSpace
    alphaBoolBlend {blendOption=Linear}
    boneToModify {boneName=dyn_jacket_MSTR_R}
    rotation=<rot 6.409 50.4309 7.6921>
   }
   n97 = RigidBody(
    ComponentPose=@n96
    OverrideWorldGravity="0,0,0"
    ComponentLinearAccScale="0.4,0.4,0.4"
    ComponentLinearVelScale="0.4,0.4,0.4"
    SimSpaceSettings=(WorldAlpha=0.5,VelocityScaleZ=0,DampingAlpha=1,MaxLinearVelocity=250,MaxAngularVelocity=3,MaxLinearAcceleration=350,MaxAngularAcceleration=500,ExternalLinearDragV=(X=0.1,Y=0.1,Z=0),ExternalLinearVelocity=(X=0,Y=0,Z=0),ExternalAngularVelocity=(X=0,Y=0,Z=0))
   ) {
    bClampLinearTranslationLimitToRefPose=true
    bOverrideWorldGravity=true
    overridePhysicsAsset=&/M_MED_InfernoFNCS_Sliding_Physics
    simulationSpace=BaseBoneSpace
    alphaBoolBlend {blendOption=Linear}
    baseBoneRef {boneName=pelvis}
    componentAppliedLinearAccClamp=<1000 1000 1000>
    componentLinearAccScale=<0.4 0.4 0.4>
    componentLinearVelScale=<0.4 0.4 0.4>
    overrideWorldGravity {z=-980}
    simSpaceSettings {
     maxAngularAcceleration=500
     maxAngularVelocity=3
     maxLinearAcceleration=350
     maxLinearVelocity=250
     velocityScaleZ=0
     worldAlpha=0.5
     externalLinearDragV=<0.1 0.1>
    }
   }
   Save cached pose 'sliding'(Pose=Component To Local(ComponentPose=@n97)) {cacheName=sliding}
  ]}
 EventGraph {nodes [
   Event BlueprintUpdateAnimation()
   > n1 = Sequence()
   > n2 = Gravity Override State Function DEPRECATED(
    CharacterPartAnimInstance=Self-Reference()
    GameTimeInSeconds=GetGameTimeInSeconds()
    GravityOverrideParams=Get Gravity Override Params()
    RBAN Params=Get RBAN Params()
   )
   > Set Gravity Override(Gravity Override=@n2)
   > Set RBAN Settings Out(RBAN Settings Out:RBANSetting Struct Structure=@n2.RBAN SettingsOut) -> (Output_Get:RBANSetting Struct Structure)
   n3 = GetOwnerSkeletalMeshComponent()
   n4 = BreakRotator(InRot=BreakTransform(InTransform=GetDeltaTransformFromRefPose(self=@n3, BoneName=thigh_l, BaseName=pelvis)).Rotation)
   Set thigh_l_rot_pitch(execute=@n1.then_1, thigh_l_rot_pitch:double=MapRangeClamped(Value=@n4.Pitch, InRangeA=0, InRangeB=33)) -> (Output_Get:double)
   > Set thigh_l_rot_yaw(thigh_l_rot_yaw:double=MapRangeClamped(Value=@n4.Yaw, InRangeA=-50, InRangeB=-100)) -> (Output_Get:double)
   n5 = BreakRotator(InRot=BreakTransform(InTransform=GetDeltaTransformFromRefPose(self=@n3, BoneName=thigh_r, BaseName=pelvis)).Rotation)
   > Set thigh_r_rot_pitch(thigh_r_rot_pitch:double=MapRangeClamped(Value=@n5.Pitch, InRangeA=0, InRangeB=33)) -> (Output_Get:double)
   > Set thigh_r_rot_yaw(thigh_r_rot_yaw:double=MapRangeClamped(Value=@n5.Yaw, InRangeA=0, InRangeB=-100)) -> (Output_Get:double)
   > Set thigh_r_rot_yaw_correction(thigh_r_rot_yaw_correction:double=MapRangeClamped(Value=@n5.Yaw, InRangeA=-50, InRangeB=-100)) -> (Output_Get:double)
   n6 = BreakRotator(InRot=BreakTransform(InTransform=GetDeltaTransformFromRefPose(self=@n3, BoneName=hand_l, BaseName=lowerarm_l)).Rotation)
   > Set hand l pitch(hand l pitch:double=MapRangeClamped(Value=@n6.Pitch, InRangeA=0, InRangeB=65)) -> (Output_Get:double)
   > Set hand l yaw(hand l yaw:double=MapRangeClamped(Value=@n6.Yaw, InRangeA=0, InRangeB=60)) -> (Output_Get:double)
   n7 = BreakRotator(InRot=BreakTransform(InTransform=GetDeltaTransformFromRefPose(self=@n3, BoneName=hand_r, BaseName=lowerarm_r)).Rotation)
   > Set hand r pitch(hand r pitch:double=MapRangeClamped(Value=@n7.Pitch, InRangeA=0, InRangeB=65)) -> (Output_Get:double)
   > Set hand r yaw(hand r yaw:double=MapRangeClamped(Value=@n7.Yaw, InRangeA=0, InRangeB=60)) -> (Output_Get:double)
   n8 = BreakTransform(InTransform=GetDeltaTransformFromRefPose(self=@n3, BoneName=clavicle_l, BaseName=spine_05))
   > Set clavicle_l_pitch(clavicle_l_pitch:double=MapRangeClamped(Value=BreakRotator(InRot=@n8.Rotation).Pitch, InRangeA=0, InRangeB=65)) -> (Output_Get:double)
   n9 = BreakTransform(InTransform=GetDeltaTransformFromRefPose(self=@n3, BoneName=clavicle_r, BaseName=spine_05))
   > Set clavicle_r_pitch(clavicle_r_pitch:double=MapRangeClamped(Value=BreakRotator(InRot=@n9.Rotation).Pitch, InRangeA=0, InRangeB=65)) -> (Output_Get:double)
  ]}
]
variables [
 clavicle_l_pitch {type=number}
 clavicle_r_pitch {type=number}
 Gravity Override {type=Vector}
 Gravity Override Params {type=GravityOverrideParamsStruct}
 hand l pitch {type=number}
 hand l yaw {type=number}
 hand r pitch {type=number}
 hand r yaw {type=number}
 RBAN Params {
  type=RBANParamsStruct
  default {
   dBNO {
    componentLinearAccScale=<0.3 0.3 0.3>
    componentLinearVelScale=<0.3 0.3 0.3>
    simSpaceSettings {
     maxAngularAcceleration=500
     maxAngularVelocity=3
     maxLinearAcceleration=350
     maxLinearVelocity=250
     velocityScaleZ=0
     worldAlpha=0.5
     externalLinearDragV=<0.1 0.1 0>
    }
   }
   dBNO_Carried {
    componentLinearAccScale=<0.3 0.3 0.3>
    componentLinearVelScale=<0.3 0.3 0.3>
    simSpaceSettings {
     maxAngularAcceleration=500
     maxAngularVelocity=3
     maxLinearAcceleration=350
     maxLinearVelocity=250
     velocityScaleZ=0
     worldAlpha=0.5
     externalLinearDragV=<0.1 0.1 0>
    }
   }
   emoteOrMelee {
    componentLinearAccScale=<0.3 0.3 0.3>
    componentLinearVelScale=<0.3 0.3 0.3>
    simSpaceSettings {
     maxAngularAcceleration=500
     maxAngularVelocity=3
     maxLinearAcceleration=350
     maxLinearVelocity=250
     velocityScaleZ=0
     worldAlpha=0.5
     externalLinearDragV=<0.1 0.1 0>
    }
   }
   front_End {
    componentLinearAccScale=<0.3 0.3 0.3>
    componentLinearVelScale=<0.3 0.3 0.3>
    simSpaceSettings {
     maxAngularAcceleration=500
     maxAngularVelocity=3
     maxLinearAcceleration=350
     maxLinearVelocity=250
     velocityScaleZ=0
     worldAlpha=0.5
     externalLinearDragV=<0.1 0.1 0>
    }
   }
   onGround_Crouch_Moving {
    componentLinearAccScale=<0.3 0.3 0.3>
    componentLinearVelScale=<0.3 0.3 0.3>
    simSpaceSettings {
     maxAngularAcceleration=500
     maxAngularVelocity=3
     maxLinearAcceleration=350
     maxLinearVelocity=250
     velocityScaleZ=0
     worldAlpha=0.5
     externalLinearDragV=<0.1 0.1 0>
    }
   }
   onGround_Falling {
    componentLinearAccScale=<0.3 0.3 0.3>
    componentLinearVelScale=<0.3 0.3 0.3>
    simSpaceSettings {
     maxAngularAcceleration=500
     maxAngularVelocity=3
     maxLinearAcceleration=350
     maxLinearVelocity=250
     velocityScaleZ=0
     worldAlpha=0.5
     externalLinearDragV=<0.1 0.1 0>
    }
   }
   onGround_Moving {
    componentLinearAccScale=<0.3 0.3 0.3>
    componentLinearVelScale=<0.3 0.3 0.3>
    simSpaceSettings {
     maxAngularAcceleration=500
     maxAngularVelocity=3
     maxLinearAcceleration=350
     maxLinearVelocity=250
     velocityScaleZ=0
     worldAlpha=0.5
     externalLinearDragV=<0.1 0.1 0>
    }
   }
   onGround_Standing {
    componentLinearAccScale=<0.3 0.3 0.3>
    componentLinearVelScale=<0.3 0.3 0.3>
    simSpaceSettings {
     maxAngularAcceleration=500
     maxAngularVelocity=3
     maxLinearAcceleration=350
     maxLinearVelocity=250
     velocityScaleZ=0
     worldAlpha=0.5
     externalLinearDragV=<0.1 0.1 0>
    }
   }
   rail_Grinding {
    componentLinearAccScale=<0.3 0.3 0.3>
    componentLinearVelScale=<0.3 0.3 0.3>
    simSpaceSettings {
     maxAngularAcceleration=500
     maxAngularVelocity=3
     maxLinearAcceleration=350
     maxLinearVelocity=250
     velocityScaleZ=0
     worldAlpha=0.5
     externalLinearDragV=<0.1 0.1 0>
    }
   }
   riding_Moving {
    componentLinearAccScale=<0.3 0.3 0.3>
    componentLinearVelScale=<0.3 0.3 0.3>
    simSpaceSettings {
     maxAngularAcceleration=500
     maxAngularVelocity=3
     maxLinearAcceleration=350
     maxLinearVelocity=250
     velocityScaleZ=0
     worldAlpha=0.5
     externalLinearDragV=<0.1 0.1 0>
    }
   }
   riding_Moving_Targeting {
    componentLinearAccScale=<0.3 0.3 0.3>
    componentLinearVelScale=<0.3 0.3 0.3>
    simSpaceSettings {
     maxAngularAcceleration=500
     maxAngularVelocity=3
     maxLinearAcceleration=350
     maxLinearVelocity=250
     velocityScaleZ=0
     worldAlpha=0.5
     externalLinearDragV=<0.1 0.1 0>
    }
   }
   riding_Standing {
    componentLinearAccScale=<0.3 0.3 0.3>
    componentLinearVelScale=<0.3 0.3 0.3>
    simSpaceSettings {
     maxAngularAcceleration=500
     maxAngularVelocity=3
     maxLinearAcceleration=350
     maxLinearVelocity=250
     velocityScaleZ=0
     worldAlpha=0.5
     externalLinearDragV=<0.1 0.1 0>
    }
   }
   skydive_DivingDown {
    componentLinearAccScale=<0.3 0.3 0.3>
    componentLinearVelScale=<0.3 0.3 0.3>
    simSpaceSettings {
     maxAngularAcceleration=500
     maxAngularVelocity=3
     maxLinearAcceleration=350
     maxLinearVelocity=250
     velocityScaleZ=0
     worldAlpha=0.5
     externalLinearDragV=<0.1 0.1 0>
    }
   }
   skydive_FloatingDown {
    componentLinearAccScale=<0.3 0.3 0.3>
    componentLinearVelScale=<0.3 0.3 0.3>
    simSpaceSettings {
     maxAngularAcceleration=500
     maxAngularVelocity=3
     maxLinearAcceleration=350
     maxLinearVelocity=250
     velocityScaleZ=0
     worldAlpha=0.5
     externalLinearDragV=<0.1 0.1 0>
    }
   }
   skydive_GoingUp {
    componentLinearAccScale=<0.3 0.3 0.3>
    componentLinearVelScale=<0.3 0.3 0.3>
    simSpaceSettings {
     maxAngularAcceleration=500
     maxAngularVelocity=3
     maxLinearAcceleration=350
     maxLinearVelocity=250
     velocityScaleZ=0
     worldAlpha=0.5
     externalLinearDragV=<0.1 0.1 0>
    }
   }
   skydive_Parachuting {
    componentLinearAccScale=<0.3 0.3 0.3>
    componentLinearVelScale=<0.3 0.3 0.3>
    simSpaceSettings {
     maxAngularAcceleration=500
     maxAngularVelocity=3
     maxLinearAcceleration=350
     maxLinearVelocity=250
     velocityScaleZ=0
     worldAlpha=0.5
     externalLinearDragV=<0.1 0.1 0>
    }
   }
   skydive_Surfing {
    componentLinearAccScale=<0.3 0.3 0.3>
    componentLinearVelScale=<0.3 0.3 0.3>
    simSpaceSettings {
     maxAngularAcceleration=500
     maxAngularVelocity=3
     maxLinearAcceleration=350
     maxLinearVelocity=250
     velocityScaleZ=0
     worldAlpha=0.5
     externalLinearDragV=<0.1 0.1 0>
    }
   }
   swimming {
    componentLinearAccScale=<0.3 0.3 0.3>
    componentLinearVelScale=<0.3 0.3 0.3>
    simSpaceSettings {
     maxAngularAcceleration=500
     maxAngularVelocity=3
     maxLinearAcceleration=350
     maxLinearVelocity=250
     velocityScaleZ=0
     worldAlpha=0.5
     externalLinearDragV=<0.1 0.1 0>
    }
   }
   tactical_Sprint {
    componentLinearAccScale=<0.3 0.3 0.3>
    componentLinearVelScale=<0.3 0.3 0.3>
    simSpaceSettings {
     maxAngularAcceleration=500
     maxAngularVelocity=3
     maxLinearAcceleration=350
     maxLinearVelocity=250
     velocityScaleZ=0
     worldAlpha=0.5
     externalLinearDragV=<0.1 0.1 0>
    }
   }
   ziplining {
    componentLinearAccScale=<0.3 0.3 0.3>
    componentLinearVelScale=<0.3 0.3 0.3>
    simSpaceSettings {
     maxAngularAcceleration=500
     maxAngularVelocity=3
     maxLinearAcceleration=350
     maxLinearVelocity=250
     velocityScaleZ=0
     worldAlpha=0.5
     externalLinearDragV=<0.1 0.1 0>
    }
   }
  }
 }
 RBAN Settings Out {
  type=RBANSettingStruct
  default {
   componentLinearAccScale=<0 0 0>
   componentLinearVelScale=<0 0 0>
   simSpaceSettings {
    dampingAlpha=1
    maxAngularAcceleration=1000
    maxAngularVelocity=3
    maxLinearAcceleration=300
    maxLinearVelocity=700
    velocityScaleZ=0
    worldAlpha=1
    externalAngularVelocity=<0 0 0>
    externalLinearDragV=<0 0.5 0>
    externalLinearVelocity=<0 0 0>
   }
  }
 }
 thigh_l_rot_pitch {type=number}
 thigh_l_rot_yaw {type=number}
 thigh_r_rot_pitch {type=number}
 thigh_r_rot_yaw {type=number}
 thigh_r_rot_yaw_correction {type=number}
]
